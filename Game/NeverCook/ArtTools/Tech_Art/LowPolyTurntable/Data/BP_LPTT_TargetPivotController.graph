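# Editor utility Blueprint that acts as the pivot controller for low-poly turntable previews, switching the debug material view on its attached preview meshes.
# Events for wireframe (with optional seams), greybox, ID islands, checker, AO, curvature and vertex color each loop over AttachedPreviewActors and call the matching material function on each LPTT static mesh actor.
# Custom Show events wrap the wireframe, checker and vertex color modes.
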EditorUtilityBlueprint /Game/NeverCook/ArtTools/Tech_Art/LowPolyTurntable/Data/BP_LPTT_TargetPivotController
graphs [
 EventGraph {nodes [
   ShowWireframe() {is=title K2Node_CustomEvent}
   > SetWireframe!(execute, self, ShowSeams?:bool=true) -> (then)
   ShowChecker() {is=title K2Node_CustomEvent}
   > SetChecker()
   ShowVertexColor() {is=title K2Node_CustomEvent}
   > SetVertexColor()
   n1 = Event SetWireframe!() -> (OutputDelegate, then, ShowSeams?:bool) {is=title K2Node_Event}
   > n2 = For Each Loop(Array:Array of LPTT Static Mesh Actor Object References=Get AttachedPreviewActors() -> (AttachedPreviewActors:Array of LPTT Static Mesh Actor Object References)) -> (Array Element:LPTT Static Mesh Actor Object Reference)
   > Material_Wireframe!(execute, self:LPTT Static Mesh Actor Object Reference=@n2, ShowSeams?:bool=@n1.ShowSeams?) -> (then)
   Event SetDefault() {is=title K2Node_Event}
   > n3 = For Each Loop(Array=Get AttachedPreviewActors())
   > Material_Greybox(self:LPTT Static Mesh Actor Object Reference=@n3)
   Event SetIDs() {is=title K2Node_Event}
   > n4 = For Each Loop(Array=Get AttachedPreviewActors())
   > Material_Islands(self:LPTT Static Mesh Actor Object Reference=@n4)
   Event SetChecker() {is=title K2Node_Event}
   > n5 = For Each Loop(Array=Get AttachedPreviewActors())
   > Material_Checker(self:LPTT Static Mesh Actor Object Reference=@n5)
   Event SetAO() {is=title K2Node_Event}
   > n6 = For Each Loop(Array=Get AttachedPreviewActors())
   > Material_AO(self:LPTT Static Mesh Actor Object Reference=@n6)
   Event SetCRV() {is=title K2Node_Event}
   > n7 = For Each Loop(Array=Get AttachedPreviewActors())
   > Material_CRV(self:LPTT Static Mesh Actor Object Reference=@n7)
   Event SetVertexColor() {is=title K2Node_Event}
   > n8 = For Each Loop(Array=Get AttachedPreviewActors())
   > Material_VertexColor(self:LPTT Static Mesh Actor Object Reference=@n8)
   Event BeginPlay()
   Event ActorBeginOverlap()
   Event Tick!() -> (OutputDelegate, then, DeltaSeconds:float)
  ]}
 UserConstructionScript {nodes [Construction Script()]}
]
variables [AttachedPreviewActors {type="array<LPTT_StaticMeshActor_C>", default []}]
