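# Editor utility widget for the LowPolyTurntable art tool that displays and controls a watched part actor.
# On construct, a details view binds to ActorRefToWatch and shows only its Data category.
# Default and Wireframe buttons cast that actor to LPTT_StaticMeshActor and apply its greybox or wireframe material, the latter without seams.
# The button row starts hidden; ToggleControlVisibility flip-flops it, setting Hidden on the first call.
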EditorUtilityWidgetBlueprint /Game/NeverCook/ArtTools/Tech_Art/LowPolyTurntable/Data/EUW_LPTT_PartData
graphs [
 EventGraph {nodes [
   "On Clicked (Default)"!() -> (OutputDelegate, then, widget:User Widget Object Reference, event:Pointer Event Structure) {is=title K2Node_ComponentBoundEvent}
   > n1 = Cast To LPTT_StaticMeshActor!(execute, Object=Get ActorRefToWatch() -> (ActorRefToWatch:Actor Object Reference)) -> (then, CastFailed, AsLPTT Static Mesh Actor:LPTT Static Mesh Actor Object Reference) {is=title K2Node_DynamicCast}
   > Material_Greybox(self:LPTT Static Mesh Actor Object Reference=@n1)
   "On Clicked (Wireframe)"!() -> (OutputDelegate, then, widget:User Widget Object Reference, event:Pointer Event Structure) {is=title K2Node_ComponentBoundEvent}
   > n2 = Cast To LPTT_StaticMeshActor(Object=Get ActorRefToWatch())
   > Material_Wireframe!(execute, self:LPTT Static Mesh Actor Object Reference=@n2, ShowSeams?:bool=false) -> (then)
   Event Construct()
   > SetObject(self=Get DetailsView_39() -> (DetailsView_39:Details View Object Reference), NewObject=Get ActorRefToWatch())
   Event PreConstruct()
   Event Tick()
  ]}
 ToggleControlVisibility {nodes [
   ToggleControlVisibility() {is=title K2Node_FunctionEntry}
   > n1 = Flip Flop!(None:Exec) -> (A:Exec, B:Exec, IsA:bool=~) {is=title K2Node_MacroInstance}
   n2 = Get HorizontalBox_0() -> (HorizontalBox_0:Horizontal Box Object Reference)
   > SetVisibility(self=@n2, InVisibility=Hidden)
   SetVisibility(execute=@n1.B, self=@n2, InVisibility=Visible)
  ]}
]
variables [
 ActorRefToWatch {type=Actor, default=null}
 ShowOverrideOptions? {type=boolean, default=true}
]
widgets {
 info {rootWidgetClass=&Overlay, widgetCount=6}
 widgets [
  Overlay_22 {widgetClassPath=&Overlay, widget=Overlay_22}
  VerticalBox_0 {widgetClassPath=&VerticalBox, parent=Overlay_22, widget=VerticalBox_0, slot {class=OverlaySlot}}
  HorizontalBox_0 {
   widgetClassPath=&HorizontalBox
   bIsVariable=true
   parent=VerticalBox_0
   widget=HorizontalBox_0
   properties {visibility=Hidden}
   slot {class=VerticalBoxSlot}
  }
  Default {
   widgetClassPath=&TA_Generic01_Button
   parent=HorizontalBox_0
   widget=Default
   properties {label=Default}
   slot {class=HorizontalBoxSlot}
  }
  Wireframe {
   widgetClassPath=&TA_Generic01_Button
   parent=HorizontalBox_0
   widget=Wireframe
   properties {label=Wireframe}
   slot {class=HorizontalBoxSlot}
  }
  DetailsView_39 {
   widgetClassPath=&DetailsView
   parent=VerticalBox_0
   widget=DetailsView_39
   properties {categoriesToShow [Data]}
   slot {class=VerticalBoxSlot}
  }
 ]
}
