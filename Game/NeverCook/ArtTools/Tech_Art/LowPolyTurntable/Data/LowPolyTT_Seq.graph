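# Level Sequence for the low-poly art turntable tool, filming Active and Ref target pivots through one fixed cine camera.
# Both pivots spin in timed phases, with a short segment that offsets and tilts them to opposite sides; Ref's spin diverges midway.
# Camera keys swap two weighted array entries for that segment and toggle post-process blend weight in later phases.
# Event triggers fire at phase boundaries; their bound functions are not visible.
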LevelSequence /Game/NeverCook/ArtTools/Tech_Art/LowPolyTurntable/Data/LowPolyTT_Seq
bindings [
 Cine Camera Actor {
  id=#1
  possessed_class=CineCameraActor
  tracks [
   {
    type=MovieSceneSpawnTrack
    sections [{type=MovieSceneSpawnSection, channels [None {type=MovieSceneScriptingBoolChannel, default=true, const [true,1]}]}]
   }
   {
    type=MovieScene3DTransformTrack
    sections [{
      type=MovieScene3DTransformSection
      channels [
       Location.X {type=MovieSceneScriptingDoubleChannel, default=-472, const [-520,1]}
       Location.Y {type=MovieSceneScriptingDoubleChannel, default=0, const [0,1]}
       Location.Z {type=MovieSceneScriptingDoubleChannel, default=84, const [170,1]}
       Rotation.X {type=MovieSceneScriptingDoubleChannel, default=0, const [0,1]}
       Rotation.Y {type=MovieSceneScriptingDoubleChannel, default=0, const [-9.5,1]}
       Rotation.Z {type=MovieSceneScriptingDoubleChannel, default=0, const [0,1]}
       Scale.X {type=MovieSceneScriptingDoubleChannel, default=1, const [1,1]}
       Scale.Y {type=MovieSceneScriptingDoubleChannel, default=1, const [1,1]}
       Scale.Z {type=MovieSceneScriptingDoubleChannel, default=1, const [1,1]}
      ]
     }]
   }
  ]
 }
 CameraComponent {
  id=#2
  parent_id=#1
  possessed_class=CineCameraComponent
  tracks [
   {
    type=MovieSceneFloatTrack
    display_name=Current Focal Length
    sections [{type=MovieSceneFloatSection, channels [None {type=MovieSceneScriptingFloatChannel, default=35, const [35,1]}]}]
   }
   {
    type=MovieSceneFloatTrack
    display_name="Manual Focus Distance (Focus Settings)"
    sections [{type=MovieSceneFloatSection, channels [None {type=MovieSceneScriptingFloatChannel, default=100000, const [100000,1]}]}]
   }
   {
    type=MovieSceneFloatTrack
    display_name=Current Aperture
    sections [{type=MovieSceneFloatSection, channels [None {type=MovieSceneScriptingFloatChannel, default=2.8, const [2.8,1]}]}]
   }
   {
    type=MovieSceneFloatTrack
    display_name="Weight (Array[0])"
    sections [{type=MovieSceneFloatSection, channels [None {type=MovieSceneScriptingFloatChannel, default=1, keys [[0,1],[179,1],[180,0],[209,0],[210,1]]}]}]
   }
   {
    type=MovieSceneFloatTrack
    display_name=Post Process Blend Weight
    sections [{type=MovieSceneFloatSection, channels [None {type=MovieSceneScriptingFloatChannel, default=1, keys [[0,1],[209,1],[210,0],[569,0],[570,1]]}]}]
   }
   {
    type=MovieSceneFloatTrack
    display_name="Weight (Array[1])"
    sections [{type=MovieSceneFloatSection, channels [None {type=MovieSceneScriptingFloatChannel, default=0, keys [[0,0],[179,0],[180,1],[209,1],[210,0]]}]}]
   }
   {
    type=MovieSceneByteTrack
    display_name=Exposure Method
    sections [{type=MovieSceneByteSection, channels [None {type=MovieSceneScriptingByteChannel, default=0}]}]
   }
  ]
 }
 BP_LPTT_PreviewActor3 {id=#3, possessed_class=BP_LPTT_PreviewActor_C, tracks []}
 ActorTarget {
  id=#4
  parent_id=#5
  possessed_class=SphereComponent
  tracks [{
    type=MovieScene3DTransformTrack
    sections [{
      type=MovieScene3DTransformSection
      channels [
       Location.X {type=MovieSceneScriptingDoubleChannel, default=0, keys [[179,0],[180,76],[209,76],[210,0]]}
       Location.Y {type=MovieSceneScriptingDoubleChannel, default=0, const [0,2]}
       Location.Z {type=MovieSceneScriptingDoubleChannel, default=0, keys [[179,0],[180,96],[209,96],[210,0]]}
       Rotation.X {type=MovieSceneScriptingDoubleChannel, default=0, keys [[179,0],[180,90],[209,90],[210,0]]}
       Rotation.Y {type=MovieSceneScriptingDoubleChannel, default=0, const [0,4]}
       Rotation.Z {
        type=MovieSceneScriptingDoubleChannel
        default=475
        thinned_from=14
        keys [[0,65],[179,475],[180,90],[209,90],[210,65],[240,450],[254,618],[255,90],[299,618],[300,90],[330,450],[570,65],[749,475]]
       }
       Scale.X {type=MovieSceneScriptingDoubleChannel, default=1}
       Scale.Y {type=MovieSceneScriptingDoubleChannel, default=1}
       Scale.Z {type=MovieSceneScriptingDoubleChannel, default=1}
      ]
     }]
   }]
 }
 ActorTarget {
  id=#6
  parent_id=#7
  possessed_class=SphereComponent
  tracks [{
    type=MovieScene3DTransformTrack
    sections [{
      type=MovieScene3DTransformSection
      channels [
       Location.X {type=MovieSceneScriptingDoubleChannel, default=0, keys [[179,0],[180,-76],[209,-76],[210,0]]}
       Location.Y {type=MovieSceneScriptingDoubleChannel, default=0, const [0,3]}
       Location.Z {type=MovieSceneScriptingDoubleChannel, default=0, keys [[179,0],[180,96],[209,96],[210,0]]}
       Rotation.X {type=MovieSceneScriptingDoubleChannel, default=0, keys [[179,0],[180,-90],[209,-90],[210,0]]}
       Rotation.Y {type=MovieSceneScriptingDoubleChannel, default=0, const [0,4]}
       Rotation.Z {
        type=MovieSceneScriptingDoubleChannel
        default=475
        keys [[0,65],[179,475],[180,90],[209,90],[210,65],[389,475],[390,65],[569,475],[570,65],[749,475]]
       }
       Scale.X {type=MovieSceneScriptingDoubleChannel, default=1}
       Scale.Y {type=MovieSceneScriptingDoubleChannel, default=1}
       Scale.Z {type=MovieSceneScriptingDoubleChannel, default=1}
      ]
     }]
   }]
 }
 ActorTargetRef {id=#5, possessed_class=BP_LPTT_TargetPivotController_C, tracks []}
 ActorTargetActive {id=#7, possessed_class=BP_LPTT_TargetPivotController_C, tracks []}
]
display_rate {denominator=1, numerator=30}
tick_resolution {denominator=1, numerator=24000}
tracks [
 {
  type=MovieSceneEventTrack
  sections [
   {
    type=MovieSceneEventTriggerSection
    channels [None {
      type=MovieSceneScriptingEventChannel
      keys [
       [-11, "<Struct 'MovieSceneEvent' (0x49e5f377940) {}>"]
       [-1, "<Struct 'MovieSceneEvent' (0x49e5f378b60) {}>"]
       [0, "<Struct 'MovieSceneEvent' (0x49e5f376f40) {}>"]
       [180, "<Struct 'MovieSceneEvent' (0x49e5f3779e0) {}>"]
       [210, "<Struct 'MovieSceneEvent' (0x49e5f377b20) {}>"]
       [390, "<Struct 'MovieSceneEvent' (0x49e5f378c00) {}>"]
       [570, "<Struct 'MovieSceneEvent' (0x49e5f378700) {}>"]
       [750, "<Struct 'MovieSceneEvent' (0x49e5f3787a0) {}>"]
      ]
     }]
   }
   {
    type=MovieSceneEventTriggerSection
    channels [None {
      type=MovieSceneScriptingEventChannel
      keys [
       [0, "<Struct 'MovieSceneEvent' (0x49e5f378a20) {}>"]
       [180, "<Struct 'MovieSceneEvent' (0x49e5f376540) {}>"]
       [570, "<Struct 'MovieSceneEvent' (0x49e5f377f80) {}>"]
      ]
     }]
   }
   {
    type=MovieSceneEventTriggerSection
    channels [None {
      type=MovieSceneScriptingEventChannel
      keys [
       [0, "<Struct 'MovieSceneEvent' (0x49e5f377da0) {}>"]
       [210, "<Struct 'MovieSceneEvent' (0x49e5f3783e0) {}>"]
       [390, "<Struct 'MovieSceneEvent' (0x49e5f378480) {}>"]
       [570, "<Struct 'MovieSceneEvent' (0x49e5f378200) {}>"]
       [750, "<Struct 'MovieSceneEvent' (0x49e5f3785c0) {}>"]
      ]
     }]
   }
   {
    type=MovieSceneEventTriggerSection
    channels [None {
      type=MovieSceneScriptingEventChannel
      keys [[0, "<Struct 'MovieSceneEvent' (0x49e5f377080) {}>"], [570, "<Struct 'MovieSceneEvent' (0x49e5f375aa0) {}>"]]
     }]
   }
   {
    type=MovieSceneEventTriggerSection
    channels [None {
      type=MovieSceneScriptingEventChannel
      keys [
       [0, "<Struct 'MovieSceneEvent' (0x49e5f376d60) {}>"]
       [180, "<Struct 'MovieSceneEvent' (0x49e5f377800) {}>"]
       [210, "<Struct 'MovieSceneEvent' (0x49e5f378980) {}>"]
      ]
     }]
   }
  ]
 }
 {type=MovieSceneCameraCutTrack, sections [{type=MovieSceneCameraCutSection, end_frame=750, start_frame=-14}]}
]
