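# Level Sequence for the zoomed variant of the LowPolyTurntable tech-art tool, presenting a reference and an active target on turntables.
# Two pivot controllers spin their targets, with a short interlude that tips them sideways while both RigCapture scene captures pull back.
# A cine camera holds a static shot while keyed post-process weights switch looks between segments.
# Event triggers at segment boundaries fire logic alongside the camera cut track.
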LevelSequence /Game/NeverCook/ArtTools/Tech_Art/LowPolyTurntable/Data/LowPolyTT_Seq_Zoomed
bindings [
 SceneCaptureComponent2D {
  id=#1
  parent_id=#2
  possessed_class=SceneCaptureComponent2D
  tracks [
   {
    type=MovieSceneFloatTrack
    display_name=Field of View
    sections [{type=MovieSceneFloatSection, channels [None {type=MovieSceneScriptingFloatChannel, default=15}]}]
   }
   {
    type=MovieScene3DTransformTrack
    sections [{
      type=MovieScene3DTransformSection
      channels [
       Location.X {type=MovieSceneScriptingDoubleChannel, default=-826, keys [[179,-564],[180,-826],[209,-826],[210,-564]]}
       Location.Y {type=MovieSceneScriptingDoubleChannel, default=0, const [0,4]}
       Location.Z {type=MovieSceneScriptingDoubleChannel, default=179, keys [[179,137],[180,179],[209,179],[210,137]]}
      ]
     }]
   }
  ]
 }
 Cine Camera Actor {
  id=#3
  possessed_class=CineCameraActor
  tracks [
   {
    type=MovieSceneSpawnTrack
    sections [{type=MovieSceneSpawnSection, channels [None {type=MovieSceneScriptingBoolChannel, default=true, const [true,1]}]}]
   }
   {
    type=MovieScene3DTransformTrack
    sections [{
      type=MovieScene3DTransformSection
      channels [
       Location.X {type=MovieSceneScriptingDoubleChannel, default=-472, const [-520,1]}
       Location.Y {type=MovieSceneScriptingDoubleChannel, default=0, const [0,1]}
       Location.Z {type=MovieSceneScriptingDoubleChannel, default=128, const [128,1]}
       Rotation.X {type=MovieSceneScriptingDoubleChannel, default=0, const [0,1]}
       Rotation.Y {type=MovieSceneScriptingDoubleChannel, default=0, const [-9.5,1]}
       Rotation.Z {type=MovieSceneScriptingDoubleChannel, default=0, const [0,1]}
       Scale.X {type=MovieSceneScriptingDoubleChannel, default=1, const [1,1]}
       Scale.Y {type=MovieSceneScriptingDoubleChannel, default=1, const [1,1]}
       Scale.Z {type=MovieSceneScriptingDoubleChannel, default=1, const [1,1]}
      ]
     }]
   }
  ]
 }
 RefPreview {id=#2, possessed_class=BP_RigCapture_C, tracks []}
 CameraComponent {
  id=#4
  parent_id=#3
  possessed_class=CineCameraComponent
  tracks [
   {
    type=MovieSceneFloatTrack
    display_name=Current Focal Length
    sections [{type=MovieSceneFloatSection, channels [None {type=MovieSceneScriptingFloatChannel, default=80, const [80,1]}]}]
   }
   {
    type=MovieSceneFloatTrack
    display_name="Manual Focus Distance (Focus Settings)"
    sections [{type=MovieSceneFloatSection, channels [None {type=MovieSceneScriptingFloatChannel, default=100000, const [100000,1]}]}]
   }
   {
    type=MovieSceneFloatTrack
    display_name=Current Aperture
    sections [{type=MovieSceneFloatSection, channels [None {type=MovieSceneScriptingFloatChannel, default=2.8, const [2.8,1]}]}]
   }
   {
    type=MovieSceneFloatTrack
    display_name="Weight (Array[0])"
    sections [{type=MovieSceneFloatSection, channels [None {type=MovieSceneScriptingFloatChannel, default=1, keys [[0,1],[179,1],[180,0],[209,0],[210,1]]}]}]
   }
   {
    type=MovieSceneFloatTrack
    display_name=Post Process Blend Weight
    sections [{type=MovieSceneFloatSection, channels [None {type=MovieSceneScriptingFloatChannel, default=1, keys [[0,1],[209,1],[210,0],[569,0],[570,1]]}]}]
   }
   {
    type=MovieSceneFloatTrack
    display_name="Weight (Array[1])"
    sections [{type=MovieSceneFloatSection, channels [None {type=MovieSceneScriptingFloatChannel, default=0, keys [[0,0],[179,0],[180,1],[209,1],[210,0]]}]}]
   }
   {
    type=MovieSceneByteTrack
    display_name=Exposure Method
    sections [{type=MovieSceneByteSection, channels [None {type=MovieSceneScriptingByteChannel, default=0}]}]
   }
  ]
 }
 BP_LPTT_PreviewActor3 {id=#5, possessed_class=BP_LPTT_PreviewActor_C, tracks []}
 ActivePreview {id=#6, possessed_class=BP_RigCapture_C, tracks []}
 ActorTarget {
  id=#7
  parent_id=#8
  possessed_class=SphereComponent
  tracks [{
    type=MovieScene3DTransformTrack
    sections [{
      type=MovieScene3DTransformSection
      channels [
       Location.X {type=MovieSceneScriptingDoubleChannel, default=0, const [0,4]}
       Location.Y {type=MovieSceneScriptingDoubleChannel, default=0, const [0,2]}
       Location.Z {type=MovieSceneScriptingDoubleChannel, default=49, keys [[179,0],[180,49],[209,49],[210,0]]}
       Rotation.X {type=MovieSceneScriptingDoubleChannel, default=0, keys [[179,0],[180,90],[209,90],[210,0]]}
       Rotation.Y {type=MovieSceneScriptingDoubleChannel, default=0, const [0,4]}
       Rotation.Z {
        type=MovieSceneScriptingDoubleChannel
        default=475
        thinned_from=14
        keys [[0,65],[179,475],[180,90],[209,90],[210,65],[240,450],[254,618],[255,90],[299,618],[300,90],[330,450],[570,65],[749,475]]
       }
       Scale.X {type=MovieSceneScriptingDoubleChannel, default=1}
       Scale.Y {type=MovieSceneScriptingDoubleChannel, default=1}
       Scale.Z {type=MovieSceneScriptingDoubleChannel, default=1}
      ]
     }]
   }]
 }
 ActorTarget {
  id=#9
  parent_id=#10
  possessed_class=SphereComponent
  tracks [{
    type=MovieScene3DTransformTrack
    sections [{
      type=MovieScene3DTransformSection
      channels [
       Location.X {type=MovieSceneScriptingDoubleChannel, default=-38, keys [[179,0],[180,-38],[209,-38],[210,0]]}
       Location.Y {type=MovieSceneScriptingDoubleChannel, default=0, const [0,3]}
       Location.Z {type=MovieSceneScriptingDoubleChannel, default=49, keys [[179,0],[180,49],[209,49],[210,0]]}
       Rotation.X {type=MovieSceneScriptingDoubleChannel, default=0, keys [[179,0],[180,-90],[209,-90],[210,0]]}
       Rotation.Y {type=MovieSceneScriptingDoubleChannel, default=0, const [0,4]}
       Rotation.Z {
        type=MovieSceneScriptingDoubleChannel
        default=475
        keys [[0,65],[179,475],[180,90],[209,90],[210,65],[389,475],[390,65],[569,475],[570,65],[749,475]]
       }
       Scale.X {type=MovieSceneScriptingDoubleChannel, default=1}
       Scale.Y {type=MovieSceneScriptingDoubleChannel, default=1}
       Scale.Z {type=MovieSceneScriptingDoubleChannel, default=1}
      ]
     }]
   }]
 }
 ActorTargetRef {id=#8, possessed_class=BP_LPTT_TargetPivotController_C, tracks []}
 SceneCaptureComponent2D {
  id=#11
  parent_id=#6
  possessed_class=SceneCaptureComponent2D
  tracks [
   {
    type=MovieSceneFloatTrack
    display_name=Field of View
    sections [{type=MovieSceneFloatSection, channels [None {type=MovieSceneScriptingFloatChannel, default=15}]}]
   }
   {
    type=MovieScene3DTransformTrack
    sections [{
      type=MovieScene3DTransformSection
      channels [
       Location.X {type=MovieSceneScriptingDoubleChannel, default=-816, keys [[179,-564],[180,-826],[209,-826],[210,-564]]}
       Location.Y {type=MovieSceneScriptingDoubleChannel, default=0, const [0,4]}
       Location.Z {type=MovieSceneScriptingDoubleChannel, default=179, keys [[179,137],[180,179],[209,179],[210,137]]}
      ]
     }]
   }
  ]
 }
 ActorTargetActive {id=#10, possessed_class=BP_LPTT_TargetPivotController_C, tracks []}
]
display_rate {denominator=1, numerator=30}
tick_resolution {denominator=1, numerator=24000}
tracks [
 {
  type=MovieSceneEventTrack
  sections [
   {
    type=MovieSceneEventTriggerSection
    channels [None {
      type=MovieSceneScriptingEventChannel
      keys [
       [-11, "<Struct 'MovieSceneEvent' (0x49e5f379740) {}>"]
       [-1, "<Struct 'MovieSceneEvent' (0x49e5f378980) {}>"]
       [0, "<Struct 'MovieSceneEvent' (0x49e5f376d60) {}>"]
       [180, "<Struct 'MovieSceneEvent' (0x49e5f375aa0) {}>"]
       [210, "<Struct 'MovieSceneEvent' (0x49e5f377080) {}>"]
       [390, "<Struct 'MovieSceneEvent' (0x49e5f378020) {}>"]
       [570, "<Struct 'MovieSceneEvent' (0x49e5f377800) {}>"]
       [750, "<Struct 'MovieSceneEvent' (0x49e5f3785c0) {}>"]
      ]
     }]
   }
   {
    type=MovieSceneEventTriggerSection
    channels [None {
      type=MovieSceneScriptingEventChannel
      keys [
       [0, "<Struct 'MovieSceneEvent' (0x49e5f378200) {}>"]
       [180, "<Struct 'MovieSceneEvent' (0x49e5f377f80) {}>"]
       [570, "<Struct 'MovieSceneEvent' (0x49e5f378d40) {}>"]
      ]
     }]
   }
   {
    type=MovieSceneEventTriggerSection
    channels [None {
      type=MovieSceneScriptingEventChannel
      keys [
       [0, "<Struct 'MovieSceneEvent' (0x49e5f3779e0) {}>"]
       [210, "<Struct 'MovieSceneEvent' (0x49e5f378b60) {}>"]
       [390, "<Struct 'MovieSceneEvent' (0x49e5f377940) {}>"]
       [570, "<Struct 'MovieSceneEvent' (0x49e5f3796a0) {}>"]
       [750, "<Struct 'MovieSceneEvent' (0x49e5f379420) {}>"]
      ]
     }]
   }
   {
    type=MovieSceneEventTriggerSection
    channels [None {
      type=MovieSceneScriptingEventChannel
      keys [[0, "<Struct 'MovieSceneEvent' (0x49e5f379060) {}>"], [570, "<Struct 'MovieSceneEvent' (0x49e5f378fc0) {}>"]]
     }]
   }
   {
    type=MovieSceneEventTriggerSection
    channels [None {
      type=MovieSceneScriptingEventChannel
      keys [
       [0, "<Struct 'MovieSceneEvent' (0x49e5f3792e0) {}>"]
       [180, "<Struct 'MovieSceneEvent' (0x49e5f379600) {}>"]
       [210, "<Struct 'MovieSceneEvent' (0x49e5f377b20) {}>"]
      ]
     }]
   }
  ]
 }
 {type=MovieSceneCameraCutTrack, sections [{type=MovieSceneCameraCutSection, end_frame=750, start_frame=-14}]}
]
