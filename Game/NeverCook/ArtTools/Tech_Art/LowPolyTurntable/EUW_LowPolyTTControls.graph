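# Editor utility widget that controls a low-poly turntable for previewing and rendering imported art assets.
# Import entries get part type, material slots and asset name detected from paths, using embedded Python, and persist in a SaveGame slot.
# FBX parts and textures feed actor sets spawned on constant and active pivot controllers.
# Controls rotate the active set, switch preview modes like wireframe, seams and AO, and launch Movie Render Queue sequences.
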EditorUtilityWidgetBlueprint /Game/NeverCook/ArtTools/Tech_Art/LowPolyTurntable/EUW_LowPolyTTControls
graphs [
 AttemptPopulateTexturePath {nodes [
   n1 = AttemptPopulateTexturePath!() -> (then, Directorypath_Path:String) {is=title K2Node_FunctionEntry}
   > n2 = Execute Python Script!(
    execute
    PythonScript="import os\r\n\r\n\r\n##recursive get all files\r\ndef recursive_getfiles(path,extension = \"\"):\r\n    local_paths = []\r\n    print(\"Grabbing files from: \", path)\r\n\r\n    for item in os.listdir(path):\r\n        fullpath = os.path.join(path, item)\r\n        if os.path.isdir(fullpath):\r\n            local_paths.append(fullpath)\r\n            local_paths.extend(recursive_getfiles(fullpath, extension+item+\"/\"))\r\n    return local_paths\r\n\r\n\r\nreturnpaths =  recursive_getfiles(startpath)"
    startpath:String=@n1
   ) -> (then, ReturnValue, returnpaths:Array of Strings) {inputs [startpath], outputs [returnpaths]}
   > n3 = Set returnpaths(returnpaths:Array of Strings=@n2.returnpaths) -> (Output_Get:Array of Strings)
   > n4 = For Each Loop(Array=@n3)
   > Branch(Condition=Contains(SearchIn=@n4, Substring=Maps))
   > Branch(Condition=IsEmpty(InString=Split(SourceString=@n4, InStr=Maps, SearchDir=FromEnd).RightS))
   Reroute Node(InputPin:String=@n4) -> (OutputPin:String)
   Set BakeMapsPath(BakeMapsPath:Array of Directory Path Structures) -> (Output_Get:Array of Directory Path Structures)
  ]}
 AutoDetectType {nodes [
   n1 = AutoDetectType!() -> (then, ImportPartStruct:LPTT Import Part Struct Structure) {is=title K2Node_FunctionEntry}
   n2 = Break LPTT Import Part Struct!(LPTT_ImportPartStruct:"LPTT Import Part Struct Structure (by ref)"=@n1) -> (
    PartPath_13_#1:File Path Structure
    AssetType_7_#2:LPTT_ImportPart_PathGen Enum
    MaterialSlots_40_#3:Array of LPTT Import Material Slot Struct Structures
    AdvancedSettings_30_#4:LPTT Import Advanced Structure
   ) {is=title K2Node_BreakStruct}
   > Set DetectedType(DetectedType:LPTT_ImportPart_PathGen Enum=@n2.AssetType_7_#2) -> (Output_Get:LPTT_ImportPart_PathGen Enum=NewEnumerator0)
   > n3 = "For Each Loop (Map)"!(
    execute:Exec
    MapPin:Map of Strings to LPTT_ImportPart_PathGen Enums=Get PartTypePathHints() -> (PartTypePathHints:Map of Strings to LPTT_ImportPart_PathGen Enums)
    BreakPin:Exec
   ) -> (then:Exec, KeyPin:String, ValuePin:LPTT_ImportPart_PathGen Enum, CompletedPin:Exec)
   n4 = Break File Path!(FilePath:"File Path Structure (by ref)"=@n2) -> (FilePath:String) {is=title K2Node_BreakStruct}
   > Branch(Condition=Contains(SearchIn=ToLower!(SourceString:String=@n4) -> (ReturnValue:String), Substring=ToLower(SourceString=@n3)))
   > n5 = Set DetectedType(DetectedType=@n3.ValuePin) -> (Output_Get=NewEnumerator0)
   > Set AutoDetected(AutoDetected:bool=true) -> (Output_Get:bool)
   > n6 = Switch on LPTT_ImportPart_PathGen!(execute, Selection:LPTT_ImportPart_PathGen Enum=@n5) -> (
    NewEnumerator0:Exec
    NewEnumerator1:Exec
    NewEnumerator2:Exec
    NewEnumerator4:Exec
    NewEnumerator7:Exec
    NewEnumerator12:Exec
    NewEnumerator13:Exec
    NewEnumerator3:Exec
    NewEnumerator8:Exec
    NewEnumerator10:Exec
   ) {is=title K2Node_SwitchEnum}
   Set DetectedDilateSize(execute=[@n6.NewEnumerator2, @n6.NewEnumerator4], DetectedDilateSize:double=4) -> (Output_Get:double)
   n7 = GetFBX!(
    execute=@n3.CompletedPin
    self
    File:File Path Structure=@n2
    PartType:String=IMPORTCHECK
    NormalsImportSetting:Enum_LPTT_NormalsImportSettings Enum=NewEnumerator0
   ) -> (then, StaticMesh:Static Mesh Object Reference)
   > Set TEMPDEBUG(TEMPDEBUG:Static Mesh Object Reference=@n7) -> (Output_Get:Static Mesh Object Reference)
   > n8 = Execute Python Script!(
    execute
    PythonScript="import unreal\r\n\r\nsm_component = unreal.StaticMeshComponent()\r\nsm_component.set_static_mesh(sm_asset)\r\nslot_names = sm_component.get_material_slot_names()"
    sm_asset:Static Mesh Object Reference=@n7
   ) -> (then, ReturnValue, slot_names:Array of Names) {inputs [sm_asset], outputs [slot_names]}
   > n9 = Set names(names:Array of Names=@n8.slot_names) -> (Output_Get:Array of Names)
   > n10 = For Each Loop(Array:Array of Names=@n9) -> (Array Element:Name)
   n11 = Get DetectedDilateSize() -> (DetectedDilateSize:double)
   Set members in LPTT Import Part Struct!(
    execute:Exec=@n10.Completed
    StructRef:"LPTT Import Part Struct Structure (by ref)"=Get ImportPartStruct() -> (ImportPartStruct:LPTT Import Part Struct Structure)
    AssetType_7_#2:LPTT_ImportPart_PathGen Enum=Get DetectedType() -> (DetectedType:LPTT_ImportPart_PathGen Enum=NewEnumerator0)
    MaterialSlots_40_#3:Array of LPTT Import Material Slot Struct Structures=Get MaterialSlots() -> (MaterialSlots:Array of LPTT Import Material Slot Struct Structures)
    AdvancedSettings_30_#4:LPTT Import Advanced Structure=Make LPTT Import Advanced!(
     WireframeDilateOverride_5_#5:int=Truncate(A=@n11)
     PivotStyle_8_#6:LPTT_PivotLocation_Enum Enum=Break LPTT Import Advanced!(LPTT_Import_Advanced:"LPTT Import Advanced Structure (by ref)"=@n2.AdvancedSettings_30_#4) -> (
      WireframeDilateOverride_5_#5:int
      PivotStyle_8_#6:LPTT_PivotLocation_Enum Enum
      ForceCreateWireframe_18_#7:bool=~
      ForceCreateColorized_UVs_21_#8:bool=~
      ForceCreateSeams_22_#9:bool=~
     ) {is=title K2Node_BreakStruct}.PivotStyle_8_#6
     ForceCreateWireframe_18_#7:bool=False
     ForceCreateColorized_UVs_21_#8:bool=False
     ForceCreateSeams_22_#9:bool=False
    ) -> (LPTT_Import_Advanced:LPTT Import Advanced Structure) {is=title K2Node_MakeStruct}
   ) -> (then:Exec, StructOut:"LPTT Import Part Struct Structure (by ref)") {is=title K2Node_SetFieldsInStruct}
   > Return Node!(
    execute
    DetectedImportPartType:LPTT_ImportPart_PathGen Enum=Get DetectedType() -> (DetectedType=NewEnumerator0)
    UnknownName:Name=None
    ModifiedImportStruct:LPTT Import Part Struct Structure=Get ImportPartStruct()
   )
   n12 = "To String (Name)"(InName=@n10)
   n13 = DetectMaterialSlotType!(execute=@n10.LoopBody, self, SlotName:String=@n12, PartType:LPTT_ImportPart_PathGen Enum=Get DetectedType() -> (DetectedType=NewEnumerator0)) -> (then, Type:LPTT_MaterialSlotType Enum=NewEnumerator0)
   > Add(
    TargetArray:Array of LPTT Import Material Slot Struct Structures=Get MaterialSlots()
    NewItem:"LPTT Import Material Slot Struct Structure (by ref)"=Make LPTT Import Material Slot Struct!(
     SlotName_23_#10:String=@n12
     MaterialType_26_#6:LPTT_MaterialSlotType Enum=@n13
     CustomMaterialType_22_#5:String
    ) -> (LPTT_Import_MaterialSlot_Struct:LPTT Import Material Slot Struct Structure) {is=title K2Node_MakeStruct}
   ) {is=title K2Node_CallArrayFunction}
  ]}
 BuildMatIDMapping {nodes [
   n1 = BuildMatIDMapping!() -> (then, Mesh:Static Mesh Object Reference, Textures:Array of Texture Object References) {is=title K2Node_FunctionEntry}
   n2 = GetMaterials!(execute, MeshComponent:Object Reference=@n1) -> (then, ReturnValue:Array of Material Interface Object References)
   n3 = GetMaterialListFromStaticMesh!(execute=@n1.then, FromStaticMeshAsset:Static Mesh Object Reference=@n1) -> (then, MaterialList:Array of Material Interface Object References, MaterialSlotNames:Array of Names)
   n4 = For Each Loop(Exec=[@n2.then, @n3.then], Array:Array of Material Interface Object References=@n3) -> (Array Element:Material Interface Object Reference)
   Return Node!(execute=@n4.Completed, Textures_Map:Map of Names to Texture Object References=Get map() -> (map:Map of Names to Texture Object References))
   > PrintString(InString=GetDisplayName(Object=@n4))
  ]}
 CleanModifiedImportParameterName {nodes [
   n1 = CleanModifiedImportParameterName!() -> (then, ParameterName:Name) {is=title K2Node_FunctionEntry}
   > n2 = Execute Python Script!(
    execute
    PythonScript="clean_name_splits = param_name.split('_')\r\n\r\nint_index = -1\r\nfor x in range(len(clean_name_splits)):\r\n    split = clean_name_splits[x]\r\n    if split.isdigit():\r\n        int_index = x\r\n        break\r\n\r\nclean_name = ''\r\n\r\nif int_index == -1 and len(clean_name_splits) == 1:\r\n    clean_name = param_name\r\nelse:\r\n    for x in range(0, int_index):\r\n        clean_name += clean_name_splits[x]\r\n        if x < int_index-1:\r\n            clean_name += \"_\""
    param_name:String="To String (Name)"(InName=@n1)
   ) -> (then, ReturnValue, clean_name:String) {inputs [param_name], outputs [clean_name]}
   > n3 = Set ModifiedImportParameterName(ModifiedImportParameterName:String=@n2.clean_name) -> (Output_Get:String)
   > Return Node!(execute, CleanParameterName:String=@n3)
  ]}
 CreateRandomFileNameExtension {nodes [
   n1 = CreateRandomFileNameExtension!() -> (then, OldFileName:String) {is=title K2Node_FunctionEntry}
   n2 = Split(SourceString=@n1, InStr=., SearchDir=FromEnd)
   > n3 = Execute Python Script!(
    execute
    PythonScript="import random\r\nimport string\r\n\r\nprint(input)\r\n\r\ncharacters = \"abcdefghijklmnopqrstuvwxyz0123456789\"\r\nvalue = ''.join(random.choices(characters,k = 10))\r\nout = f'{input}{value}'"
    input:String=@n2
   ) -> (then, ReturnValue, out:String) {inputs [input], outputs [out]}
   n4 = Append!(A=@n3.out, B=., C:String=@n2.RightS) -> (ReturnValue)
   > PrintString(InString=@n4)
   > Return Node!(execute, NewName:String=@n4)
  ]}
 DetectMaterialSlotType {nodes [
   n1 = DetectMaterialSlotType!() -> (then, SlotName:String, PartType:LPTT_ImportPart_PathGen Enum=NewEnumerator0) {is=title K2Node_FunctionEntry}
   > n2 = Switch on LPTT_ImportPart_PathGen(Selection=@n1.PartType)
   Return Node!(execute=[@n2.NewEnumerator0, @n2.NewEnumerator1, @n2.NewEnumerator2, @n2.NewEnumerator4], Type:LPTT_MaterialSlotType Enum=NewEnumerator0)
   Return Node(execute=[@n2.NewEnumerator10, @n2.NewEnumerator7, @n2.NewEnumerator8, @n3.Default], Type=NewEnumerator0)
   n3 = Switch on String!(execute:Exec=@n2.NewEnumerator3, Selection:String=@n1) -> (Default:Exec, Eyes:Exec) {pinNames [Eyes]}
   Return Node(execute=@n3.Eyes, Type=NewEnumerator1)
  ]}
 DetermineAssetName {nodes [
   DetermineAssetName() {is=title K2Node_FunctionEntry}
   > n1 = Sequence()
   > n2 = For Each Loop!(Exec, Array:Array of LPTT Import Part Struct Structures=Get ImportPaths() -> (ImportPaths:Array of LPTT Import Part Struct Structures)) -> (
    LoopBody
    Array Element_PartPath_13_#1:File Path Structure
    Array Element_AssetType_7_#2:LPTT_ImportPart_PathGen Enum
    Array Element_MaterialSlots_40_#3:Array of LPTT Import Material Slot Struct Structures
    Array Element_AdvancedSettings_30_#4:LPTT Import Advanced Structure
    Array Index
    Completed
   )
   n3 = Split(SourceString=Break File Path(FilePath=@n2), InStr=artsource, SearchDir=FromStart)
   > n4 = Branch(Condition=IsEmpty(InString=@n3.RightS))
   n5 = Split(SourceString=@n3.RightS, InStr=model, SearchDir=FromStart)
   n6 = Branch(execute=@n4.else, Condition=IsEmpty(InString=@n5))
   n7 = Split(SourceString=@n5, InStr=/, SearchDir=FromEnd)
   n8 = Branch(execute=@n6.else, Condition=IsEmpty(InString=@n7))
   n9 = Split(SourceString=@n7, InStr=/, SearchDir=FromEnd)
   n10 = Branch(execute=@n8.else, Condition=IsEmpty(InString=@n9.RightS))
   Add(
    execute=@n10.else
    TargetArray:Array of Strings=Get AllAssetNames() -> (AllAssetNames:Array of Strings)
    NewItem:"String (by ref)"=@n9.RightS
   )
   > Add Unique(TargetArray=Get UniqueAssetNames() -> (UniqueAssetNames:Array of Strings), NewItem=@n9.RightS)
   n11 = Get UniqueAssetNames()
   n12 = Branch(execute=@n2.Completed, Condition="Equal (Integer)"(A=Length(TargetArray:Array of Strings=@n11), B=1))
   > Set AssetName(AssetName:String="Get (a copy)"(Array=Get UniqueAssetNames())) -> (Output_Get:String)
   > Return Node!(execute)
   n13 = For Each Loop!(Exec=@n12.else, Array:Array of Strings=Get AllAssetNames()) -> (LoopBody, Array Element:String, Array Index, Completed)
   n14 = "For Each Loop (Map)"(execute=@n13.Completed, MapPin:Map of Strings to Integers=Get NameCounts() -> (NameCounts:Map of Strings to Integers)) -> (ValuePin:int)
   > Branch(Condition="integer > integer"(A=@n14.ValuePin, B=Get LargestNameCount() -> (LargestNameCount:int=0)))
   > Set LargestNameCount(LargestNameCount:int=@n14.ValuePin) -> (Output_Get:int=0)
   > Set LargestName(LargestName:String=@n14) -> (Output_Get:String)
   n15 = Branch(
    execute=@n13.LoopBody
    Condition=Contains!(TargetMap:Map of Strings to Integers=Get NameCounts(), Key:"String (by ref)"=@n13) -> (ReturnValue)
   )
   n16 = Get NameCounts()
   > Add!(
    execute
    TargetMap:Map of Strings to Integers=Get NameCounts()
    Key:"String (by ref)"=@n13
    Value:"Integer (by ref)"=int + int(A=Find(TargetMap:Map of Strings to Integers=@n16, Key=@n13) -> (Value:int))
   ) -> (then) {is=title K2Node_CallFunction}
   Add(execute=@n15.else, TargetMap=Get NameCounts(), Key=@n13, Value=1)
   Set AssetName(execute=@n1.then_1, AssetName=Get LargestName() -> (LargestName:String))
  ]}
 EventGraph {nodes [
   ResetPreviewOverrides()
   > n1 = DelayUntilNextTick!(execute, WorldContextObject:Object Reference) -> (then)
   HandlePreviewOverrides!(
    execute=[@n1.then, @n65.then]
    self
    Mode:Name=Seams
    Target:TA Base Editor Button Widget Object Reference=Get Seams() -> (Seams:TA Generic 01 Button Object Reference)
   ) -> (then)
   n2 = ResetPreviewOverrides()
   HandlePreviewOverrides(execute=[@n2.then, @n63.then], Mode=Greybox, Target=Get Greybox() -> (Greybox:TA Generic 01 Button Object Reference))
   n3 = ResetPreviewOverrides()
   HandlePreviewOverrides(execute=[@n3.then, @n64.then], Mode=VC, Target=Get VertexColor() -> (VertexColor:TA Generic 01 Button Object Reference))
   n4 = Get ModifiedImportParameterName() -> (ModifiedImportParameterName:String)
   GetModifiedImportStruct!(execute, self, ModifiedParameterName:String=@n4) -> (then)
   > UpdateImportPropertyString()
   n5 = "On Clicked (Render)"!() -> (OutputDelegate, then, widget:User Widget Object Reference, event:Pointer Event Structure) {is=title K2Node_ComponentBoundEvent}
   ShowSpotterSpheres?!(execute=[@n5.then, @n69.then], self, Show?:bool=false) -> (then)
   > ResetRotation()
   > DelayUntilNextTick()
   > n6 = Branch(Condition=Get FocusHead() -> (FocusHead:bool))
   n7 = Get OutputDirectory() -> (OutputDirectory:Directory Path Structure)
   > RenderSequence!(
    execute
    self
    Job Name:String=FocusHead
    Sequence:Soft Object Path Structure=Get MRQ_Sequence_HeadFocus() -> (MRQ_Sequence_HeadFocus:Soft Object Path Structure)
    Output Directory:Directory Path Structure=@n7
   ) -> (then, ReturnValue:Movie Pipeline PIEExecutor Object Reference)
   n8 = Branch(execute=@n6.else, Condition=Get RenderZoomed() -> (RenderZoomed:bool))
   > RenderSequence(Job Name=ZoomedRender, Sequence=Get MRQ_Sequence_Zoomed() -> (MRQ_Sequence_Zoomed:Soft Object Path Structure), Output Directory=@n7)
   n9 = RenderSequence(execute=@n8.else, Job Name=StandardLowPoly, Sequence=Get MRQ_Sequence() -> (MRQ_Sequence:Soft Object Path Structure), Output Directory=@n7)
   n10 = ShowHelpers!() -> (OutputDelegate, then, FinishedJob:Movie Pipeline Executor Job Object Reference, bSuccess:bool=~) {is=title K2Node_CustomEvent}
   Bind Event to On Individual Job Finished Delegate(execute=@n9.then, self:Movie Pipeline PIEExecutor Object Reference=@n9, Delegate=@n10) {is=title K2Node_AddDelegate}
   n11 = "On Clicked (Importtotextures)"!() -> (OutputDelegate, then, widget:User Widget Object Reference, event:Pointer Event Structure) {is=title K2Node_ComponentBoundEvent}
   n12 = Sequence(execute=[@n11.then, @n67.then])
   SetValue!(
    execute=[@n66.then, @n12.then_0]
    self:Slider Object Reference=Get Rotation_Slider() -> (Rotation_Slider:Slider Object Reference)
    InValue:float=0
   ) -> (then)
   n13 = Get ActiveTargetRotator() -> (ActiveTargetRotator:BP LPTT Target Pivot Controller Object Reference)
   n14 = Get Actor Rotation!(self:Actor Object Reference=@n13) -> (ReturnValue_Roll:float, ReturnValue_Pitch:float, ReturnValue_Yaw:float) {is=title K2Node_CallFunction}
   > Set Actor Rotation!(
    execute
    self:Actor Object Reference=@n13
    NewRotation_Roll:float=@n14
    NewRotation_Pitch:float=@n14.ReturnValue_Pitch
    NewRotation_Yaw:float=0
    bTeleportPhysics:bool=false
   ) -> (then, ReturnValue:bool) {is=title K2Node_CallFunction}
   Clear(
    execute=@n12.then_1
    TargetArray:Array of LPTT Part Data Struct Structures=Get CompiledPartData() -> (CompiledPartData:Array of LPTT Part Data Struct Structures)
   )
   > n15 = GetGeneralTextures!(execute, self) -> (then, Textures:Array of Texture 2D Object References)
   > n16 = For Each Loop(Array:Array of LPTT Import Part Struct Structures=Get ImportPaths()) -> (Array Element:LPTT Import Part Struct Structure)
   n17 = Break LPTT Import Part Struct!(LPTT_ImportPartStruct=@n16) -> (
    PartPath_13_#1
    AssetType_7_#2
    MaterialSlots_40_#3
    AdvancedSettings_30_#4_WireframeDilateOverride_5_#5:int
    AdvancedSettings_30_#4_PivotStyle_8_#6:LPTT_PivotLocation_Enum Enum
    AdvancedSettings_30_#4_ForceCreateWireframe_18_#7:bool=~
    AdvancedSettings_30_#4_ForceCreateColorized_UVs_21_#8:bool=~
    AdvancedSettings_30_#4_ForceCreateSeams_22_#9:bool=~
   )
   > n18 = Branch(Condition=IsEmpty(InString=Break File Path(FilePath=@n17)))
   > PrintString(InString=Error - Entry Skipped due to missing mesh filepath)
   n19 = Sequence!(execute=@n18.else) -> (then_0, then_1, then_2:Exec)
   > n20 = GetFBX(
    File=@n17
    PartType=Enum to String!(Enumerator:LPTT_ImportPart_PathGen Enum=@n17.AssetType_7_#2) -> (ReturnValue:String)
    NormalsImportSetting=Get Normals Import Setting() -> (Normals Import Setting:Enum_LPTT_NormalsImportSettings Enum=NewEnumerator0)
   )
   Add!(
    execute=@n19.then_2
    TargetArray:Array of LPTT Part Data Struct Structures=Get CompiledPartData()
    NewItem_FBXPath_29_#11:File Path Structure=@n17
    NewItem_Mesh_5_#12:Static Mesh Object Reference=@n20
    NewItem_Textures_30_#13:Array of Texture 2D Object References=@n15
    NewItem_MaterialSlotNames_24_#14:Array of LPTT Import Material Slot Struct Structures=@n17.MaterialSlots_40_#3
    NewItem_PartType_25_#15:LPTT_ImportPart_PathGen Enum=@n17.AssetType_7_#2
    NewItem_WireframeDilateSize_12_#16:int=@n17.AdvancedSettings_30_#4_WireframeDilateOverride_5_#5
    NewItem_PivotStyle_15_#17:LPTT_PivotLocation_Enum Enum=@n17.AdvancedSettings_30_#4_PivotStyle_8_#6
    NewItem_ForceCreateWireframe_32_#18:bool=@n17.AdvancedSettings_30_#4_ForceCreateWireframe_18_#7
    NewItem_ForceCreateColorized_UVs_34_#19:bool=@n17.AdvancedSettings_30_#4_ForceCreateColorized_UVs_21_#8
    NewItem_ForceCreateSeams_36_#20:bool=@n17.AdvancedSettings_30_#4_ForceCreateSeams_22_#9
   ) -> (then, ReturnValue) {is=title K2Node_CallArrayFunction}
   n21 = Get SpawnedActors() -> (SpawnedActors:Array of LPTT Static Mesh Actor Object References)
   n22 = For Each Loop(Exec=[@n89.then, @n16.Completed], Array:Array of LPTT Static Mesh Actor Object References=@n21) -> (Array Element:LPTT Static Mesh Actor Object Reference)
   Clear(execute=@n22.Completed, TargetArray:Array of LPTT Static Mesh Actor Object References=@n21)
   > n23 = GetAllActorsOfClass!(
    execute
    WorldContextObject:Object Reference
    ActorClass:Actor Class Reference=/Game/NeverCook/ArtTools/Tech_Art/LowPolyTurntable/Data/LPTT_StaticMeshActor.LPTT_StaticMeshActor_C
   ) -> (then, OutActors:Array of LPTT Static Mesh Actor Object References)
   > n24 = For Each Loop(Array=@n23)
   > Destroy Actor(self=@n24)
   n25 = GetCurrentLevel!(execute=@n24.Completed, self:Level Editor Subsystem Object Reference=Get LevelEditorSubsystem()) -> (then, ReturnValue:Level Object Reference)
   > n26 = Sequence()
   > n27 = GetAllActorsOfClassWithTag!(
    execute
    WorldContextObject:Object Reference=@n25
    ActorClass:Actor Class Reference=/Game/NeverCook/ArtTools/Tech_Art/LowPolyTurntable/Data/BP_LPTT_TargetPivotController.BP_LPTT_TargetPivotController_C
    Tag:Name=ConstantRef
   ) -> (then, OutActors:Array of BP LPTT Target Pivot Controller Object References)
   n28 = "Get (a copy)"(Array:Array of BP LPTT Target Pivot Controller Object References=@n27) -> (Output:BP LPTT Target Pivot Controller Object Reference)
   n29 = Get ActorTarget!(self:BP LPTT Target Pivot Controller Object Reference=@n28) -> (ActorTarget:Sphere Collision Object Reference)
   > n30 = Is Valid(InputObject=@n29)
   PrintString(execute=@n30.Is Not Valid, InString="~~~~~~~~~~~~~~~~~~~~~~MISSSING ACTOR TARGET")
   n31 = Get!() -> (
    TransformOverrides_Location_8_#21:Vector=~
    TransformOverrides_Rotation_7_#22:Rotator=~
    TransformOverrides_Scale_6_#23:double=~
   )
   n32 = Get SpawnedActors()
   n33 = Spawn ActorSet!(
    execute=@n30.Is Valid
    self
    SpawnedActorArrayRef:Array of LPTT Static Mesh Actor Object References=@n32
    Spawn Transform Location:Vector=vector + vector(A=Get World Location!(self:Scene Component Object Reference=@n29) -> (ReturnValue:Vector) {is=title K2Node_CallFunction}, B=@n31)
    ShowControls?:bool=false
   ) -> (then, ThisActorSet:Array of LPTT Static Mesh Actor Object References)
   > Set AttachedPreviewActors!(
    execute
    AttachedPreviewActors:Array of LPTT Static Mesh Actor Object References=@n33
    self:BP LPTT Target Pivot Controller Object Reference=@n28
   ) -> (then, Output_Get:Array of LPTT Static Mesh Actor Object References)
   > n34 = For Each Loop(Array=@n33)
   > Attach Actor To Component!(
    execute
    self:Actor Object Reference=@n34
    Parent:Scene Component Object Reference=@n29
    SocketName:Name=None
    LocationRule:EAttachmentRule Enum=KeepWorld
    RotationRule:EAttachmentRule Enum=KeepWorld
    ScaleRule:EAttachmentRule Enum=KeepWorld
    bWeldSimulatedBodies:bool=TRUE
   ) -> (then, ReturnValue:bool)
   n35 = GetAllActorsOfClassWithTag(
    execute=@n26.then_1
    WorldContextObject=@n25
    ActorClass=&BP_LPTT_TargetPivotController_C
    Tag=ActiveRef
   )
   n36 = "Get (a copy)"(Array=@n35)
   > Set ActiveTargetRotator(ActiveTargetRotator:BP LPTT Target Pivot Controller Object Reference=@n36) -> (Output_Get:BP LPTT Target Pivot Controller Object Reference)
   n37 = Get ActorTarget(self=@n36)
   > n38 = Spawn ActorSet(SpawnedActorArrayRef=@n32, Spawn Transform Location=vector + vector(A=Get World Location(self=@n37), B=@n31), ShowControls?=true)
   > Set AttachedPreviewActors(AttachedPreviewActors=@n38, self=@n36)
   > n39 = For Each Loop(Array=@n38)
   > Attach Actor To Component(self=@n39, Parent=@n37, SocketName=None, LocationRule=KeepWorld, RotationRule=KeepWorld, ScaleRule=KeepWorld, bWeldSimulatedBodies=TRUE)
   > AddActorTag!(execute, Actor:Actor Object Reference=@n39, Tag:"Name (by ref)"=MakeLiteralName(Value=Active), __WorldContext:Object Reference) -> (then)
   SetIsEnabled(execute=@n26.then_2, self=Get Render() -> (Render:TA Generic 01 Button Object Reference), bInIsEnabled=true)
   Destroy Actor(execute=@n22.LoopBody, self=@n22)
   "On Clicked (AoMap)"!() -> (OutputDelegate, then, widget:User Widget Object Reference, event:Pointer Event Structure) {is=title K2Node_ComponentBoundEvent}
   > HandlePreviewOverrides(Mode=AO, Target=Get AoMap() -> (AoMap:TA Generic 01 Button Object Reference))
   n40 = "On Clicked (ClearFloor)"!() -> (OutputDelegate, then, widget:User Widget Object Reference, event:Pointer Event Structure) {is=title K2Node_ComponentBoundEvent}
   n41 = Get SpawnedActors()
   n42 = For Each Loop(Exec=[@n40.then, @n70.then], Array=@n41)
   Clear(execute=@n42.Completed, TargetArray=@n41)
   > n43 = GetAllActorsOfClass(ActorClass=&LPTT_StaticMeshActor_C)
   > n44 = For Each Loop(Array=@n43)
   > Destroy Actor(self=@n44)
   Destroy Actor(execute=@n42.LoopBody, self=@n42)
   "On Property Changed (RenderSettings)"!() -> (OutputDelegate, then, PropertyName:Name=None) {is=title K2Node_ComponentBoundEvent}
   > ShowSpotterSpheres?(Show?=true)
   > n45 = For Each Loop(Array=Get SpawnedActors())
   n46 = Get() -> (
    TransformOverrides_Location_8_#21=~
    TransformOverrides_Rotation_7_#22=~
    TransformOverrides_Scale_6_#23=~
   )
   > Set Actor Relative Transform!(
    execute
    self:Actor Object Reference=@n45
    NewRelativeTransform_Location:Vector=@n46
    NewRelativeTransform_Rotation:Rotator=@n46.TransformOverrides_Rotation_7_#22
    NewRelativeTransform_Scale:Vector=MakeVector(
     X=@n46.TransformOverrides_Scale_6_#23
     Y=@n46.TransformOverrides_Scale_6_#23
     Z=@n46.TransformOverrides_Scale_6_#23
    )
    bSweep:bool=false
    bTeleport:bool=false
   ) -> (then, SweepHitResult:Hit Result Structure) {is=title K2Node_CallFunction}
   n47 = "On Property Changed (Import_Paths)"!() -> (OutputDelegate, then, PropertyName:Name=None) {is=title K2Node_ComponentBoundEvent}
   > n48 = CleanModifiedImportParameterName!(execute, self, ParameterName:Name=@n47.PropertyName) -> (then, CleanParameterName:String) {is=title K2Node_CallFunction}
   DetermineAssetName(execute=[@n48.then, @n58.then]) {is=title K2Node_CallFunction}
   > n49 = Branch(Condition=Contains Item(TargetArray:Array of Strings=Get ImportPropertyIgnores() -> (ImportPropertyIgnores:Array of Strings), ItemToFind:"String (by ref)"=@n4))
   n50 = CreateSaveGameObject!(
    execute=[@n49.then, @n52.Completed]
    SaveGameClass:Save Game Class Reference=/Game/NeverCook/ArtTools/Tech_Art/LowPolyTurntable/Data/LPTT_SaveGameData.LPTT_SaveGameData_C
   ) -> (then, ReturnValue:LPTT Save Game Data Object Reference)
   > SaveData!(
    execute
    self:LPTT Save Game Data Object Reference=@n50
    ImportPaths:Array of LPTT Import Part Struct Structures=Get ImportPaths()
    TextureMapsPath:Array of Directory Path Structures=Get BakeMapsPath() -> (BakeMapsPath:Array of Directory Path Structures)
   ) -> (then)
   > SaveGameToSlot!(execute, SaveGameObject:Save Game Object Reference=@n50, SlotName:String=Get SaveDataSlot() -> (SaveDataSlot:String), UserIndex:int=0) -> (then, ReturnValue:bool)
   n51 = Get ImportPaths()
   n52 = For Each Loop(Exec=[@n68.then, @n49.else], Array:Array of LPTT Import Part Struct Structures=@n51) -> (Array Element:LPTT Import Part Struct Structure)
   n53 = Break LPTT Import Part Struct!(LPTT_ImportPartStruct=@n52) -> (
    PartPath_13_#1_FilePath:String
    AssetType_7_#2
    MaterialSlots_40_#3
    AdvancedSettings_30_#4_WireframeDilateOverride_5_#5
    AdvancedSettings_30_#4_PivotStyle_8_#6
    AdvancedSettings_30_#4_ForceCreateWireframe_18_#7=~
    AdvancedSettings_30_#4_ForceCreateColorized_UVs_21_#8=~
    AdvancedSettings_30_#4_ForceCreateSeams_22_#9=~
   )
   > n54 = Branch(Condition=IsEmpty(InString=@n53))
   n55 = Switch on LPTT_ImportPart_PathGen(execute=@n54.else, Selection=@n53.AssetType_7_#2)
   > n56 = AutoDetectType!(execute, self, ImportPartStruct:LPTT Import Part Struct Structure=@n52) -> (
    then
    DetectedImportPartType:LPTT_ImportPart_PathGen Enum=NewEnumerator0
    UnknownName:Name=None
    ModifiedImportStruct:LPTT Import Part Struct Structure
   ) {is=title K2Node_CallFunction}
   > Set Array Elem!(
    execute
    TargetArray:Array of LPTT Import Part Struct Structures=@n51
    Index:int=@n52.Array Index
    Item:"LPTT Import Part Struct Structure (by ref)"=@n56.ModifiedImportStruct
    bSizeToFit:bool=false
   ) -> (then)
   > SetSuffixString!(
    execute
    self
    Type:LPTT_ImportPart_PathGen Enum=@n56
    ImportEntry:"LPTT Import Part Struct Structure (by ref)"=@n52
    UnknownName:Name=@n56.UnknownName
    Dilate:int=@n53.AdvancedSettings_30_#4_WireframeDilateOverride_5_#5
    Advanced Settings Pivot Style:LPTT_PivotLocation_Enum Enum=@n53.AdvancedSettings_30_#4_PivotStyle_8_#6
   ) -> (then)
   SetSuffixString(
    execute=[@n55.NewEnumerator1, @n55.NewEnumerator2, @n55.NewEnumerator3, @n55.NewEnumerator4, @n55.NewEnumerator7, @n55.NewEnumerator8]
    Type=@n53.AssetType_7_#2
    ImportEntry=@n52
    UnknownName=None
    Dilate=@n53.AdvancedSettings_30_#4_WireframeDilateOverride_5_#5
    Advanced Settings Pivot Style=@n53.AdvancedSettings_30_#4_PivotStyle_8_#6
   )
   > UpdateMaterialSlots!(execute, self, ImportPartStruct:"LPTT Import Part Struct Structure (by ref)"=@n52) -> (then)
   n57 = "On Property Changed (Import_Settings)"!() -> (OutputDelegate, then, PropertyName:Name=None) {is=title K2Node_ComponentBoundEvent}
   > n58 = CleanModifiedImportParameterName(ParameterName=@n57.PropertyName)
   n59 = "On Clicked (ClearImportValues)"!() -> (OutputDelegate, then, widget:User Widget Object Reference, event:Pointer Event Structure) {is=title K2Node_ComponentBoundEvent}
   Clear(execute=[@n59.then, @n71.then], TargetArray:Array of LPTT Import Part Struct Structures=Get ImportPaths())
   > Clear(TargetArray:Array of Directory Path Structures=Get BakeMapsPath())
   > Set AssetName()
   > n60 = CreateSaveGameObject(SaveGameClass=&LPTT_SaveGameData_C)
   > SaveData(self=@n60, ImportPaths=Get ImportPaths(), TextureMapsPath=Get BakeMapsPath())
   > SaveGameToSlot(SaveGameObject=@n60, SlotName=Get SaveDataSlot(), UserIndex=0)
   n61 = "On Value Changed (Rotation_Slider)"!() -> (OutputDelegate, then, Value:float) {is=title K2Node_ComponentBoundEvent}
   n62 = Get Actor Rotation(self=@n13)
   > Set Actor Rotation(self=@n13, NewRotation_Roll=@n62, NewRotation_Pitch=@n62.ReturnValue_Pitch, NewRotation_Yaw=@n61.Value, bTeleportPhysics=false)
   n63 = "On Clicked (Greybox)"!() -> (OutputDelegate, then, widget:User Widget Object Reference, event:Pointer Event Structure) {is=title K2Node_ComponentBoundEvent}
   n64 = "On Clicked (VertexColor)"!() -> (OutputDelegate, then, widget:User Widget Object Reference, event:Pointer Event Structure) {is=title K2Node_ComponentBoundEvent}
   n65 = "On Clicked (Seams)"!() -> (OutputDelegate, then, widget:User Widget Object Reference, event:Pointer Event Structure) {is=title K2Node_ComponentBoundEvent}
   "On Clicked (Wireframe)"!() -> (OutputDelegate, then, widget:User Widget Object Reference, event:Pointer Event Structure) {is=title K2Node_ComponentBoundEvent}
   > HandlePreviewOverrides(Mode=Wireframe, Target=Get Wireframe() -> (Wireframe:TA Generic 01 Button Object Reference))
   "On Clicked (Checker)"!() -> (OutputDelegate, then, widget:User Widget Object Reference, event:Pointer Event Structure) {is=title K2Node_ComponentBoundEvent}
   > HandlePreviewOverrides(Mode=Checker, Target=Get Checker() -> (Checker:TA Generic 01 Button Object Reference))
   n66 = ResetRotation() {is=title K2Node_CustomEvent}
   n67 = TriggerImportTextures() {is=title K2Node_CustomEvent}
   n68 = TriggerGenerateImportGroups() {is=title K2Node_CustomEvent}
   n69 = TriggerRender() {is=title K2Node_CustomEvent}
   n70 = TriggerClearFloor() {is=title K2Node_CustomEvent}
   n71 = TriggerClearImportValues() {is=title K2Node_CustomEvent}
   Event PreConstruct()
   > SetObject(
    self=[
     Get DV_Textures() -> (DV_Textures:Details View Object Reference)
     Get Import_Settings() -> (Import_Settings:Details View Object Reference)
     Get Import_Paths() -> (Import_Paths:Details View Object Reference)
     Get DebugDetailView() -> (DebugDetailView:Details View Object Reference)
     Get AssetNameField() -> (AssetNameField:Details View Object Reference)
     Get RenderSettings() -> (RenderSettings:Details View Object Reference)
    ]
    NewObject=Self-Reference()
   )
   > ExecuteConsoleCommand(Command=Slate.bAllowThrottling 0)
   Event Construct()
   > ShowSpotterSpheres?(Show?=false)
   > SetIsEnabled(self=Get Render(), bInIsEnabled=false)
   > Branch(Condition=Get ShouldLoadSavedSettings() -> (ShouldLoadSavedSettings:bool))
   > n72 = LoadGameFromSlot!(execute, SlotName:String=Get SaveDataSlot(), UserIndex:int=0) -> (then, ReturnValue:Save Game Object Reference)
   > n73 = Cast To LPTT_SaveGameData!(execute, Object=@n72) -> (then, CastFailed, AsLPTT Save Game Data:LPTT Save Game Data Object Reference) {is=title K2Node_DynamicCast}
   > n74 = LoadData!(execute, self:LPTT Save Game Data Object Reference=@n73) -> (then, ImportPaths:Array of LPTT Import Part Struct Structures, TextureMapsPath:Array of Directory Path Structures)
   > Set ImportPaths(ImportPaths:Array of LPTT Import Part Struct Structures=@n74) -> (Output_Get:Array of LPTT Import Part Struct Structures)
   > Set BakeMapsPath(BakeMapsPath=@n74.TextureMapsPath)
   > DetermineAssetName()
   n75 = Get!() -> (ImportDirectory_Path:String)
   n76 = Execute Python Script!(
    execute
    PythonScript="import Developers.ChrisMorrison.TurnTables.LPTT_Importer as LPTT_Importer\r\nimport importlib\r\nimportlib.reload(LPTT_Importer)\r\n\r\nfiles = []\r\nout_fbx = []\r\nout_texture = []\r\nout_partcollections = []\r\nout_partcollections_csv = []\r\npart_type = []\r\n\r\n##Data analisys variables\r\nuniqueparts_fbx = []\r\nuniqueparts_texture = []\r\n\r\n(out_fbx, out_texture, out_partcollections_csv, part_type) = LPTT_Importer.LPTT_Import(dir)"
    dir:String=@n75
   ) -> (then, ReturnValue, out_partcollections_csv:Array of Strings, PartType:Array of Strings) {inputs [dir], outputs [out_partcollections_csv, PartType]}
   n77 = Set ImportSets(execute=[@n76.then, @n85.then], ImportSets:Array of Strings=@n76.out_partcollections_csv) -> (Output_Get:Array of Strings)
   > Set PartTypes(PartTypes:Array of Strings=@n76.PartType) -> (Output_Get:Array of Strings)
   n78 = Get PartPathGroups() -> (PartPathGroups:Array of LPTT Filename Import Struct Structures)
   > Resize!(execute, TargetArray:Array of LPTT Filename Import Struct Structures=@n78, Size:int=Length(TargetArray=@n77)) -> (then) {is=title K2Node_CallArrayFunction}
   > n79 = For Each Loop(Array:Array of Strings=@n77) -> (Array Element:String)
   n80 = "Get (a ref)"(Array:Array of LPTT Filename Import Struct Structures=@n78, Dimension 1=@n79.Array Index) -> (Output:"LPTT Filename Import Struct Structure (by ref)")
   n81 = Get PartTypes() -> (PartTypes:Array of Strings)
   n82 = ParseIntoArray(SourceString=@n79, Delimiter=|)
   n83 = "Get (a copy)"(Array:Array of Strings=@n82) -> (Output:String)
   > "Set LPTT Filename Import Struct Structure (by ref)"!(
    execute:Exec
    Target:"LPTT Filename Import Struct Structure (by ref)"=@n80
    Value:LPTT Filename Import Struct Structure=Make LPTT Filename Import Struct!(
     Mesh_12_#24:File Path Structure=Make File Path!(FilePath:String=@n83) -> (FilePath:File Path Structure) {is=title K2Node_MakeStruct}
     Textures_11_#25:Array of File Path Structures
     Type_15_#26:String="Get (a copy)"(Array=@n81, Dimension 1=@n79.Array Index)
    ) -> (LPTT_FilenameImportStruct:LPTT Filename Import Struct Structure) {is=title K2Node_MakeStruct}
   ) -> (then:Exec) {is=title K2Node_VariableSetRef}
   > n84 = For Loop(FirstIndex=1, LastIndex=int - int(A=Length(TargetArray=@n82)))
   > Add Unique(
    TargetArray:Array of File Path Structures=Break LPTT Filename Import Struct!(LPTT_FilenameImportStruct:"LPTT Filename Import Struct Structure (by ref)"=@n80) -> (
     Mesh_12_#24:File Path Structure
     Textures_11_#25:Array of File Path Structures
     Type_15_#26:String
    ) {is=title K2Node_BreakStruct}.Textures_11_#25
    NewItem:"File Path Structure (by ref)"=Make File Path(FilePath="Get (a copy)"(Array=@n82, Dimension 1=@n84))
   )
   Execute Python Script!(
    execute
    PythonScript="import os\r\n\r\nfiles = []\r\nout_fbx = []\r\nout_texture = []\r\nout_partcollections = []\r\nout_partcollections_csv = []\r\npart_type = []\r\n\r\n##Data analisys variables\r\nuniqueparts_fbx = []\r\nuniqueparts_texture = []\r\n\r\n\r\n##strips out the partname for an fbx path\r\ndef parseFBXfortype(file):\r\n\treturn  file.split(\"_\")[-1].split(\".\")[0]\r\n\r\n##strips out the partname for a texture path, assumes _<part type>_<texture type>.png for formating\r\ndef parseTexturefortype(file, shift = -2):\r\n\treturn  file.split(\"_\")[shift].split(\".\")[0]\r\n\r\n\r\n\r\nprint(\"Begin asset import...\")\r\n\r\nfor item in os.listdir(dir):\r\n\tif os.path.isfile(os.path.join(dir,item)):\r\n\t\tfiles.append(item)\r\n\t\tprint(item)\r\n\r\nfor file in files:\r\n\tif \".png\" in file.lower():\r\n\t\tprint(\"Texture found: \", file)\r\n\t\tout_texture.append(file)\r\n\t\tif parseTexturefortype(file) not in uniqueparts_texture:\r\n\t\t\tuniqueparts_texture.append(parseTexturefortype(file))\r\n\telif \".fbx\" in file.lower():\r\n\t\tprint(\"FBX found: \", file)\r\n\t\tout_fbx.append(file)\r\n\t\tif parseFBXfortype(file) not in uniqueparts_fbx:\r\n\t\t\tuniqueparts_fbx.append(parseFBXfortype(file))\r\n\t\telse:\r\n\t\t\tprint(\"Warning! Duplicate FBX type labels found!\")\r\n\telse:\r\n\t\tprint(\"Unsupported filetype found: \", file)\r\n\r\n\r\n##check and throw warning if the unique type count doesn't match\r\nif len(uniqueparts_fbx) != len(uniqueparts_texture):\r\n\tprint(\"Warning! Mismatch in FBX and Texture Part Types! Extra Parts detected?\")\r\n\tprint(uniqueparts_fbx, \" : \", uniqueparts_texture)\r\n\r\n\r\n##find body association\r\nbodytype = \"null\"\r\nfor entry in uniqueparts_fbx:\r\n\t##find Body based on naming of beginning of texture names\r\n\tif out_texture[0].split(\"_\")[0] in entry:\r\n\t\tprint(\"Body found: \", entry)\r\n\t\tbodytype = entry\r\n\r\n##create sets for each type\r\nfor index, entry in enumerate(uniqueparts_fbx):\r\n\t##entry 0 is always the FBX filepath\r\n\ttypelist = []\r\n\ttypelist.append(out_fbx[index])\r\n\t\r\n\r\n\t##handle Body check\r\n\tif entry == bodytype:\r\n\t\tentry = 'Body'\r\n\t\tprint(\"Body detected, setting to: \", entry)\r\n\r\n\t##set the type output\r\n\tpart_type.append(entry)\r\n\r\n\tfor texture_entry in out_texture:\r\n\t\tif entry in texture_entry:\r\n\t\t\tprint(\"part added: \", texture_entry)\r\n\t\t\ttypelist.append(texture_entry)\r\n\r\n\tout_partcollections.append(typelist)\r\n\r\n##create output CSV Strings\r\nfor entry in out_partcollections:\r\n\tprint(\"-------\")\r\n\tout_csv_string = \"\"\r\n\tfor subentry in entry:\r\n\t\tprint(subentry)\r\n\t\tout_csv_string += os.path.join(dir,subentry)+\"|\"\r\n\tout_partcollections_csv.append(out_csv_string)\r\nprint(\"-------\")\r\n\r\nfor entry in out_partcollections_csv:\r\n\tprint(entry)\r\n\tprint(\"-------\")\r\n\r\n\r\n\r\nprint(\"...End asset import.\")"
    dir=@n75
   ) -> (then, ReturnValue, out_fbx:Wildcard, out_texture:Wildcard, out_partcollections_csv:Wildcard, PartType:Wildcard) {inputs [dir], outputs [out_fbx, out_texture, out_partcollections_csv, PartType]}
   > Set InputArray_FBX(InputArray_FBX:Array of Strings) -> (Output_Get:Array of Strings)
   > n85 = Set InputArray_Texture(InputArray_Texture:Array of Strings) -> (Output_Get:Array of Strings)
   n86 = Branch(Condition=IsEmpty(InString=Get()))
   Clear(execute=@n86.else, TargetArray:Array of LPTT Filename Import Struct Structures=Get PartPathGroups())
   n87 = Get PartDataSubWidgets() -> (PartDataSubWidgets:Array of EUW LPTT Part Data Object References)
   > n88 = For Each Loop(Array:Array of EUW LPTT Part Data Object References=@n87) -> (Array Element:EUW LPTT Part Data Object Reference)
   n89 = Clear(execute=@n88.Completed, TargetArray:Array of EUW LPTT Part Data Object References=@n87)
   RemoveFromParent(execute=@n88.LoopBody, self=@n88)
   MakeLiteralName(Value=ActiveRef)
   MakeLiteralName(Value=ConstantRef)
   ShowSpotterSpheres?(execute=@n10.then, Show?=true)
   UpdateImportPropertyString()
   Event Tick()
   Reroute Node(InputPin:int=@n52.Array Index) -> (OutputPin:int)
  ]}
 GetAllTextures {nodes [
   n1 = GetAllTextures!() -> (
    then
    DirectoryPath_Path:String
    GeneralTextures:Array of Texture Object References
    TypeSuffix:String
    MaterialSlots:Array of LPTT Import Material Slot Struct Structures
   ) {is=title K2Node_FunctionEntry}
   > n2 = Branch(Condition=IsEmpty(InString=@n1))
   n3 = For Each Loop(Exec=[@n6.Completed, @n2.then], Array:Array of Texture Object References=@n1.GeneralTextures) -> (Array Element:Texture Object Reference)
   Return Node!(
    execute=@n3.Completed
    Textures:Array of Texture Object References=Get CompiledTextures() -> (CompiledTextures:Array of Texture Object References)
   )
   > Branch(Condition=Contains!(SearchIn:String=Get Object Path String(Object=@n3), Substring:String=@n1.TypeSuffix, bUseCase:bool, bSearchFromEnd:bool) -> (ReturnValue))
   > Add Unique(TargetArray:Array of Texture Object References=Get CompiledTextures(), NewItem:"Texture Object Reference (by ref)"=@n3)
   n4 = Execute Python Script!(
    execute=@n2.else
    PythonScript="import Developers.ChrisMorrison.TurnTables.LPTT_Importer as LPTT_Importer\r\nimport importlib\r\nimportlib.reload(LPTT_Importer)\r\n\r\ntextures = []\r\n\r\nlocal_files=LPTT_Importer.recursive_getfiles(directory)\r\n\r\nfor entry in local_files:\r\n    if \".png\" in entry.filename.lower():\r\n        textures.append(entry.fullpath)        "
    Directory:String=@n1
   ) -> (then, ReturnValue, textures:Array of Strings) {inputs [Directory], outputs [textures]}
   > n5 = Set Filepaths(Filepaths:Array of Strings=@n4.textures) -> (Output_Get:Array of Strings)
   > n6 = For Each Loop(Array:Array of Strings=@n5) -> (Array Element:String)
   > n7 = ImportAsset!(execute, self, InFileName:String=@n6, Destination Name:String, NormalsImportSetting:Enum_LPTT_NormalsImportSettings Enum=NewEnumerator0) -> (then, OutImportedObjects:Array of Object References)
   > n8 = Cast To Texture!(execute, Object="Get (a copy)"(Array:Array of Object References=@n7) -> (Output:Object Reference)) -> (then, CastFailed, AsTexture:Texture Object Reference) {is=title K2Node_DynamicCast}
   > Set NeverStream!(execute, NeverStream:bool=true, self:Streamable Render Asset Object Reference=@n8) -> (then, Output_Get:bool)
   n9 = Get MipGenSettings() -> (MipGenSettings:TextureMipGenSettings Enum=TMGS_FromTextureGroup)
   > n10 = Execute Python Script!(
    execute
    PythonScript="import unreal\r\n\r\nobject.set_editor_property(\"mip_gen_settings\",value)"
    object:Texture Object Reference=@n8
    value:TextureMipGenSettings Enum=@n9
   ) -> (then, ReturnValue) {inputs [object,value]}
   Branch(execute=[@n10.then, @n11.then, @n12.then], Condition=Contains(SearchIn=Get Object Path String(Object=@n8), Substring=@n1.TypeSuffix))
   > Add Unique(TargetArray=Get CompiledTextures(), NewItem=@n8)
   n11 = SetEditorProperty(Object=@n8, PropertyName=mip_gen_settings, PropertyValue:TextureMipGenSettings Enum=@n9)
   n12 = Set MipGenSettings!(execute, MipGenSettings:TextureMipGenSettings Enum=TMGS_FromTextureGroup, self:Texture Object Reference=@n8) -> (then, Output_Get:TextureMipGenSettings Enum=TMGS_FromTextureGroup)
   For Each Loop!(Exec, Array:Array of LPTT Import Material Slot Struct Structures=@n1.MaterialSlots) -> (
    LoopBody
    Array Element_SlotName_23_#10:String
    Array Element_MaterialType_26_#6:LPTT_MaterialSlotType Enum
    Array Element_CustomMaterialType_22_#5:String
    Array Index
    Completed
   )
  ]}
 GetAllTexturesOld {nodes [
   n1 = GetAllTexturesOld!() -> (then, DirectoryPath_Path:String, GeneralTextures:Array of Texture 2D Object References, TypeSuffix:String) {is=title K2Node_FunctionEntry}
   > n2 = Branch(Condition=IsEmpty(InString=@n1))
   n3 = For Each Loop!(Exec=[@n6.Completed, @n2.then], Array:Array of Texture 2D Object References=@n1.GeneralTextures) -> (LoopBody, Array Element:Texture 2D Object Reference, Array Index, Completed)
   Return Node(
    execute=@n3.Completed
    Textures:Array of Texture 2D Object References=Get CompiledTextures() -> (CompiledTextures:Array of Texture 2D Object References)
   )
   > Branch(Condition=Contains(SearchIn=Get Object Path String(Object=@n3), Substring=@n1.TypeSuffix))
   > Add Unique(TargetArray:Array of Texture 2D Object References=Get CompiledTextures(), NewItem:"Texture 2D Object Reference (by ref)"=@n3)
   n4 = Execute Python Script!(
    execute=@n2.else
    PythonScript="import Developers.ChrisMorrison.TurnTables.LPTT_Importer as LPTT_Importer\r\nimport importlib\r\nimportlib.reload(LPTT_Importer)\r\n\r\ntextures = []\r\n\r\nlocal_files=LPTT_Importer.recursive_getfiles(directory)\r\n\r\nfor entry in local_files:\r\n    if \".png\" in entry.filename.lower():\r\n        textures.append(entry.fullpath)        "
    Directory:String=@n1
   ) -> (then, ReturnValue, textures:Array of Strings) {inputs [Directory], outputs [textures]}
   > n5 = Set Filepaths(Filepaths=@n4.textures)
   > n6 = For Each Loop(Array:Array of Strings=@n5) -> (Array Element:String)
   > n7 = For Each Loop with Break!(Exec:Exec, Array:Array of Strings=Get TextureExtensions() -> (TextureExtensions:Array of Strings), Break:Exec) -> (LoopBody:Exec, Array Element:String, Array Index:int, Completed:Exec)
   > Set FoundIgnoreSuffix(FoundIgnoreSuffix:bool=false) -> (Output_Get:bool)
   > Branch(Condition=EndsWith(SourceString=GetBaseFilename(InPath=@n6, bRemovePath=true), InSuffix=@n7))
   > n8 = For Each Loop(Array=Get TextureExtensionsIgnore() -> (TextureExtensionsIgnore:Array of Strings))
   n9 = Branch(execute=@n8.Completed, Condition=Get FoundIgnoreSuffix() -> (FoundIgnoreSuffix:bool))
   n10 = Branch(execute=@n9.else, Condition=Get Quickload() -> (Quickload:bool))
   n11 = GetBaseFilename(InPath=@n6, bRemovePath=true)
   n12 = GetBaseFilename(InPath=@n6, bRemovePath=true)
   > n13 = Branch(Condition=OR Boolean(A=EndsWith(SourceString=@n12, InSuffix=_N), B=EndsWith(SourceString=@n11, InSuffix=_ao)))
   n14 = ImportAsset(execute=[@n10.else, @n13.then], InFileName=@n6, NormalsImportSetting=NewEnumerator0)
   > n15 = Cast To Texture2D(Object="Get (a copy)"(Array=@n14))
   > Set NeverStream(NeverStream=true, self=@n15)
   > Execute Python Script!(
    execute
    PythonScript="import unreal\r\n\r\nobject.set_editor_property(\"mip_gen_settings\",value)"
    object:Texture 2D Object Reference=@n15
    value:TextureMipGenSettings Enum=Get MipGenSettings() -> (MipGenSettings=TMGS_FromTextureGroup)
   ) -> (then, ReturnValue) {inputs [object,value]}
   > Branch(Condition=Contains(SearchIn=Get Object Path String(Object=@n15), Substring=@n1.TypeSuffix))
   > Add Unique(TargetArray=Get CompiledTextures(), NewItem=@n15)
   Branch(execute=@n8.LoopBody, Condition=EndsWith(SourceString=GetBaseFilename(InPath=@n6, bRemovePath=true), InSuffix=@n8))
   > Set FoundIgnoreSuffix(FoundIgnoreSuffix=true)
   Reroute Node(InputPin:Array of Texture 2D Object References=@n1.GeneralTextures) -> (OutputPin:Array of Texture 2D Object References)
  ]}
 GetFBX {nodes [
   n1 = Get FileName() -> (FileName:String)
   n2 = GetFBX!() -> (then, File_FilePath:String, PartType:String, NormalsImportSetting:Enum_LPTT_NormalsImportSettings Enum=NewEnumerator0) {is=title K2Node_FunctionEntry}
   n3 = Append(A=Split(SourceString=Split(SourceString=@n1, InStr=/, SearchDir=FromEnd).RightS, InStr=., SearchDir=FromEnd), B=_, C=@n2.PartType)
   n4 = DoesAssetExist(AssetPath=Append!(A=/Temp/, B=@n3) -> (ReturnValue))
   n5 = IsEmpty(InString=Get FileName())
   > n6 = Branch(Condition=AND Boolean(A=NOT Boolean(A=@n5), B=@n4))
   > n7 = CreateRandomFileNameExtension!(execute, self, OldFileName:String=Get FileName()) -> (then, NewName:String) {is=title K2Node_CallFunction}
   > n8 = Set FileName(FileName:String=@n7) -> (Output_Get:String)
   n9 = ImportAsset(execute=[@n11.then, @n6.else, @n8.then, @n13.then], InFileName=@n2, Destination Name=@n3, NormalsImportSetting=@n2.NormalsImportSetting)
   > n10 = Cast To StaticMesh(Object="Get (a copy)"(Array=@n9))
   > Add(
    TargetMap:Map of Strings to Static Mesh Object References=Get ImportedFBX() -> (ImportedFBX:Map of Strings to Static Mesh Object References)
    Key=ToLower(SourceString=Get FileName())
    Value:"Static Mesh Object Reference (by ref)"=@n10
   ) {is=title K2Node_CallFunction}
   > Return Node!(execute, StaticMesh:Static Mesh Object Reference=@n10)
   n11 = Execute Python Script!(
    execute
    PythonScript="import Developers.ChrisMorrison.TurnTables.LPTT_Importer as LPTT_Importer\r\nimport importlib\r\nimportlib.reload(LPTT_Importer)\r\n\r\nfbxpath = \"\"\r\n\r\nlocal_files=LPTT_Importer.recursive_getfiles(directory)\r\nfor entry in local_files:\r\n    if \".fbx\" in entry.filename.lower():\r\n        fbxpath=entry.fullpath"
    Directory:Wildcard
   ) -> (then, ReturnValue, fbxpath:Wildcard) {inputs [Directory], outputs [fbxpath]}
   n12 = Find(TargetMap:Map of Strings to Static Mesh Object References=Get ImportedFBX(), Key=ToLower(SourceString=@n2)) -> (Value:Static Mesh Object Reference)
   Branch(Condition=@n12.ReturnValue)
   > Return Node(StaticMesh=@n12)
   Get FileName()
   n13 = Set FileName(execute=@n2.then, FileName=@n2)
  ]}
 GetGeneralTextures {nodes [
   GetGeneralTextures() {is=title K2Node_FunctionEntry}
   > n1 = For Each Loop(Array:Array of Directory Path Structures=Get BakeMapsPath()) -> (Array Element:Directory Path Structure)
   Return Node!(
    execute=@n1.Completed
    Textures:Array of Texture 2D Object References=Get LocalTextures() -> (LocalTextures:Array of Texture 2D Object References)
   )
   > n2 = GetAllTexturesOld!(execute, self, DirectoryPath:Directory Path Structure=@n1, GeneralTextures:Array of Texture 2D Object References, TypeSuffix:String) -> (then, Textures:Array of Texture 2D Object References) {is=title K2Node_CallFunction}
   > Append Array!(execute, TargetArray:Array of Texture 2D Object References=Get LocalTextures(), SourceArray:Array of Texture 2D Object References=@n2) -> (then)
  ]}
 GetModifiedImportPartStruct {nodes [GetModifiedImportPartStruct() {is=title K2Node_FunctionEntry}]}
 GetModifiedImportStruct {nodes [
   n1 = GetModifiedImportStruct!() -> (then, ModifiedParameterName:String) {is=title K2Node_FunctionEntry}
   > n2 = For Each Loop with Break(Array:Array of LPTT Import Part Struct Structures=Get ImportPaths()) -> (Array Element:LPTT Import Part Struct Structure)
   > Execute Python Script!(
    execute
    PythonScript="import json\r\n\r\nlower_param_name = param_name[0].lower()\r\nlower_param_name = lower_param_name + param_name[1:]\r\n\r\ntry:\r\n    json_str = json_list[index]\r\nexcept:\r\n    json_str = None\r\n\r\nif json_str:\r\n    json_obj = json.loads(json_str)\r\nelse:\r\n    json_obj = None\r\n\r\nif json_obj:\r\n    if param_name == 'MaterialType':\r\n        print(json_obj)\r\n        adv_sets = json_obj.get('advancedSettings', {})\r\n        print(adv_sets)\r\n        mat_slots = adv_sets.get('materialSlots', [])\r\n        print(mat_slots)\r\n        for x in range(len(mat_slots)):\r\n            mat_dict = mat_slots[x]\r\n            print(mat_dict)\r\n            old_val = mat_dict.get(lower_param_name)\r\n            print(old_val)\r\n\r\nprint(\"GGGGGGGGGGGGGGGGGG\")\r\nprint(import_struct.get_editor_property(\"advancedSettings\"))"
    param_name:String=@n1
    index:int=@n2.Array Index
    json_list:Array of Strings=Get ImportStructsStringLast() -> (ImportStructsStringLast:Array of Strings)
    import_struct:LPTT Import Part Struct Structure=@n2
   ) -> (then, ReturnValue) {inputs [param_name, index, json_list, import_struct]}
   Branch(Condition=StartsWith(SourceString=Get ModifiedParameterName() -> (ModifiedParameterName:String), InPrefix=CustomMaterialType))
   n3 = Break LPTT Import Part Struct!(LPTT_ImportPartStruct=@n2) -> (
    PartPath_13_#1:File Path Structure
    AssetType_7_#2
    MaterialSlots_40_#3
    AdvancedSettings_30_#4:LPTT Import Advanced Structure
   )
   > n4 = For Each Loop with Break(Array:Array of LPTT Import Material Slot Struct Structures=@n3.MaterialSlots_40_#3) -> (Array Element:LPTT Import Material Slot Struct Structure)
   Break LPTT Import Advanced(LPTT_Import_Advanced=@n3.AdvancedSettings_30_#4) -> (
    ForceCreateWireframe_18_#7=~
    ForceCreateColorized_UVs_21_#8=~
    ForceCreateSeams_22_#9=~
   )
   Execute Python Script!(
    execute
    PythonScript="print('Hello World!')"
    new_val:String=Break LPTT Import Material Slot Struct!(LPTT_Import_MaterialSlot_Struct:"LPTT Import Material Slot Struct Structure (by ref)"=@n4) -> (
     SlotName_23_#10:String
     MaterialType_26_#6:LPTT_MaterialSlotType Enum
     CustomMaterialType_22_#5:String
    ) {is=title K2Node_BreakStruct}.CustomMaterialType_22_#5
    json_str:"String (by ref)"="Get (a ref)"(Array:Array of Strings=Get ImportStructsStringLast(), Dimension 1=@n2.Array Index) -> (Output:"String (by ref)")
   ) -> (then, ReturnValue, is_diff:Wildcard) {inputs [new_val,json_str], outputs [is_diff]}
   For Each Loop(Array:Array of Strings=Get ImportStructsStringCurrent() -> (ImportStructsStringCurrent:Array of Strings)) -> (Array Element:String)
  ]}
 GetWireframeTextures {nodes [
   GetWireframeTextures() {is=title K2Node_FunctionEntry}
   Return Node!(execute, WireframeTextures:Array of Texture Object References=Get WireTextures() -> (WireTextures:Array of Texture Object References))
  ]}
 HandlePreviewOverrides {nodes [
   n1 = HandlePreviewOverrides!() -> (then, Mode:Name, Target:TA Base Editor Button Widget Object Reference) {is=title K2Node_FunctionEntry}
   > n2 = For Each Loop(Array:Array of Widget Object References=GetAllChildren(self=Get WrapBox_0() -> (WrapBox_0:Wrap Box Object Reference))) -> (Array Element:Widget Object Reference)
   > n3 = Cast To EUW_LPTT_PartData!(execute, Object=@n2) -> (then, CastFailed, AsEUW LPTT Part Data:EUW LPTT Part Data Object Reference) {is=title K2Node_DynamicCast}
   > n4 = Cast To LPTT_StaticMeshActor!(execute, Object=Get ActorRefToWatch!(self:EUW LPTT Part Data Object Reference=@n3) -> (ActorRefToWatch:Actor Object Reference)) -> (then, CastFailed, AsLPTT Static Mesh Actor:LPTT Static Mesh Actor Object Reference) {is=title K2Node_DynamicCast}
   > n5 = Switch on Name!(execute:Exec, Selection:Name=@n1) -> (Default:Exec, Greybox:Exec, Wireframe:Exec, Seams:Exec, Checker:Exec, VC:Exec, AO:Exec, Reset:Exec) {is=title K2Node_SwitchName, pinNames [Greybox,Wireframe,Seams,Checker,VC,AO,Reset]}
   Material_AO(execute=@n5.AO, self:LPTT Static Mesh Actor Object Reference=@n4)
   Material_Checker(execute=@n5.Checker, self:LPTT Static Mesh Actor Object Reference=@n4)
   Material_Greybox(execute=@n5.Greybox, self:LPTT Static Mesh Actor Object Reference=@n4)
   Material_Wireframe!(execute=@n5.Seams, self:LPTT Static Mesh Actor Object Reference=@n4, ShowSeams?:bool=true) -> (then)
   Material_VertexColor(execute=@n5.VC, self:LPTT Static Mesh Actor Object Reference=@n4)
   Material_Wireframe(execute=@n5.Wireframe, self=@n4, ShowSeams?=false)
  ]}
 ImportAsset {nodes [
   n1 = ImportAsset!() -> (then, InFileName:String, Destination Name:String, NormalsImportSetting:Enum_LPTT_NormalsImportSettings Enum=NewEnumerator0) {is=title K2Node_FunctionEntry}
   > Set FileName(FileName=@n1)
   > n2 = Construct Object from Class(Class=/Game/NeverCook/ArtTools/Tech_Art/LowPolyTurntable/Data/BP_LPTT_InterchangeBP.BP_LPTT_InterchangeBP_C) -> (ReturnValue:BP LPTT Interchange BP Object Reference)
   > Set NormalsImportSetting!(
    execute
    NormalsImportSetting:Enum_LPTT_NormalsImportSettings Enum=@n1.NormalsImportSetting
    self:BP LPTT Interchange BP Object Reference=@n2
   ) -> (then, Output_Get:Enum_LPTT_NormalsImportSettings Enum=NewEnumerator0)
   > n3 = CreateSourceData!(execute, InFileName:String=Get FileName()) -> (then, ReturnValue:Interchange Source Data Object Reference)
   > n4 = GetInterchangeManagerScripted!(execute) -> (then, ReturnValue:Interchange Manager Object Reference)
   > n5 = Branch(Condition=IsEmpty(InString=Get AssetName() -> (AssetName:String)))
   n6 = Make Import Asset Parameters!(
    ReimportAsset:Object Reference
    ReimportSourceIndex:int=-1
    bIsAutomated:bool=true
    bFollowRedirectors:bool=False
    OverridePipelines:Array of Soft Object Path Structures=Make Array!(
     "[0]":Soft Object Path Structure=Get CustomInterchange() -> (CustomInterchange:Soft Object Path Structure)
     "[1]":Soft Object Path Structure=GetSoftObjectPath(Object=@n2)
    ) -> (Array:Array of Soft Object Path Structures)
    ImportLevel:Level Object Reference
    DestinationName:String=@n1.Destination Name
    bReplaceExisting:bool=True
    bForceShowDialog:bool=False
   ) -> (ImportAssetParameters:Import Asset Parameters Structure) {is=title K2Node_MakeStruct}
   > n7 = ImportAsset!(
    execute
    self:Interchange Manager Object Reference=@n4
    ContentPath:String=/Game/Temp/
    SourceData:Interchange Source Data Object Reference=@n3
    ImportAssetParameters:"Import Asset Parameters Structure (by ref)"=@n6
   ) -> (then, OutImportedObjects, ReturnValue:bool) {is=title K2Node_CallFunction}
   > Return Node!(execute, OutImportedObjects:Array of Object References=@n7)
   n8 = ImportAsset(
    execute=@n5.else
    self=@n4
    ContentPath=Append!(A=&Temp/, B=Get AssetName(), C:String=/) -> (ReturnValue)
    SourceData=@n3
    ImportAssetParameters=@n6
   )
   > Return Node(OutImportedObjects=@n8)
   ReimportAsset!(
    execute
    self:Interchange Manager Object Reference=@n4
    ObjectToReimport:Object Reference
    ImportAssetParameters:"Import Asset Parameters Structure (by ref)"
   ) -> (then, OutImportedObjects:Array of Object References, ReturnValue:bool)
   n9 = CreateAssetData!(InAsset=Get SelfRef() -> (SelfRef:Object Reference), bAllowBlueprintClass=true) -> (
    ReturnValue_PackageName:Name
    ReturnValue_PackagePath:Name
    ReturnValue_AssetName:Name
    ReturnValue_AssetClassPath:Top Level Asset Path Structure
   )
   Append(A=Split(SourceString="To String (Name)"(InName=@n9), InStr=/, SearchDir=FromEnd), B=/PerLoadAssets/)
  ]}
 RenderSequence {nodes [
   n1 = Cast To MoviePipelineConsoleVariableSetting!(execute, Object:Wildcard) -> (then, CastFailed, AsMovie Pipeline Console Variable Setting:Movie Pipeline Console Variable Setting Object Reference) {is=title K2Node_DynamicCast}
   n2 = Get ConsoleVariables!(self:Movie Pipeline Console Variable Setting Object Reference=@n1) -> (ConsoleVariables:"Map of Strings to Float (single-precision)s")
   n3 = Keys!(execute, TargetMap:"Map of Strings to Float (single-precision)s"=@n2) -> (then, Keys:Array of Strings)
   > For Each Loop(Array:Array of Strings=@n3) -> (Array Element:String)
   n4 = RenderSequence!() -> (then, Job Name:String, Sequence:Soft Object Path Structure, Output Directory:Directory Path Structure) {is=title K2Node_FunctionEntry}
   > n5 = Sequence()
   n6 = Get MoviePipelineQueueSubsystem() -> (ReturnValue:Movie Pipeline Queue Subsystem Object Reference) {is=title K2Node_GetEditorSubsystem}
   n7 = GetQueue!(self:Movie Pipeline Queue Subsystem Object Reference=@n6) -> (ReturnValue:Movie Pipeline Queue Object Reference)
   > n8 = For Each Loop(Array:Array of Movie Pipeline Executor Job Object References=GetJobs!(self:Movie Pipeline Queue Object Reference=@n7) -> (ReturnValue:Array of Movie Pipeline Executor Job Object References)) -> (Array Element:Movie Pipeline Executor Job Object Reference)
   > DeleteJob!(execute, self:Movie Pipeline Queue Object Reference=@n7, InJob:Movie Pipeline Executor Job Object Reference=@n8) -> (then)
   n9 = AllocateNewJob!(
    execute=@n5.then_1
    self:Movie Pipeline Queue Object Reference=@n7
    InJobType:Movie Pipeline Executor Job Class Reference=/Script/MovieRenderPipelineCore.MoviePipelineExecutorJob
   ) -> (then, ReturnValue:Movie Pipeline Executor Job Object Reference)
   > Set JobName!(execute, JobName:String=@n4, self:Movie Pipeline Executor Job Object Reference=@n9) -> (then, Output_Get:String)
   > Set Sequence!(execute, Sequence:Soft Object Path Structure=@n4.Sequence, self:Movie Pipeline Executor Job Object Reference=@n9) -> (then, Output_Get:Soft Object Path Structure)
   > Set Map!(execute, Map:Soft Object Path Structure=Get Map() -> (Map:Soft Object Path Structure), self:Movie Pipeline Executor Job Object Reference=@n9) -> (then, Output_Get:Soft Object Path Structure)
   > SetConfiguration!(
    execute
    self:Movie Pipeline Executor Job Object Reference=@n9
    InPreset:Movie Pipeline Primary Config Object Reference=Get MRQPresets() -> (MRQPresets:Movie Pipeline Primary Config Object Reference)
   ) -> (then)
   n10 = GetConfiguration!(self:Movie Pipeline Executor Job Object Reference=@n9) -> (ReturnValue:Movie Pipeline Primary Config Object Reference)
   > n11 = For Each Loop(Array:Array of Movie Pipeline Setting Object References=GetAllSettings!(self:Movie Pipeline Primary Config Object Reference=@n10, bIncludeDisabledSettings:bool=false, bIncludeTransientSettings:bool=false) -> (ReturnValue:Array of Movie Pipeline Setting Object References)) -> (Array Element:Movie Pipeline Setting Object Reference)
   n12 = FindOrAddSettingByClass!(
    execute=@n11.Completed
    self:Movie Pipeline Config Base Object Reference=@n10
    InClass:Movie Pipeline Setting Class Reference=/Script/MovieRenderPipelineCore.MoviePipelineOutputSetting
    bIncludeDisabledSettings:bool=false
    bExactMatch:bool=false
   ) -> (then, ReturnValue:Movie Pipeline Output Setting Object Reference)
   > Set OutputDirectory!(execute, OutputDirectory:Directory Path Structure=@n4.Output Directory, self:Movie Pipeline Output Setting Object Reference=@n12) -> (then, Output_Get:Directory Path Structure)
   n13 = Cast To MoviePipelineAntiAliasingSetting!(execute=@n11.LoopBody, Object=@n11) -> (then, CastFailed, AsMovie Pipeline Anti Aliasing Setting:Movie Pipeline Anti Aliasing Setting Object Reference) {is=title K2Node_DynamicCast}
   > n14 = Switch on ECheckBoxState!(execute, Selection:ECheckBoxState Enum=Undetermined) -> (Unchecked:Exec, Checked:Exec, Undetermined:Exec) {is=title K2Node_SwitchEnum}
   Set bOverrideAntiAliasing!(execute=[@n14.Checked, @n14.Undetermined], bOverrideAntiAliasing:bool=false, self:Movie Pipeline Anti Aliasing Setting Object Reference=@n13) -> (then, Output_Get:bool)
   Set bOverrideAntiAliasing(execute=@n14.Unchecked, bOverrideAntiAliasing=true, self=@n13)
   > Set AntiAliasingMethod!(execute, AntiAliasingMethod:EAntiAliasingMethod Enum=AAM_None, self:Movie Pipeline Anti Aliasing Setting Object Reference=@n13) -> (then, Output_Get:EAntiAliasingMethod Enum=AAM_None)
   > PrintString(InString="~~~~~~~~~~~~~~~~~ Overriding AntiAliasing to None")
   n15 = RenderQueueWithExecutor!(
    execute=@n5.then_2
    self:Movie Pipeline Queue Subsystem Object Reference=@n6
    InExecutorType:Movie Pipeline Executor Base Class Reference=/Script/MovieRenderPipelineEditor.MoviePipelinePIEExecutor
   ) -> (then, ReturnValue:Movie Pipeline PIEExecutor Object Reference)
   > Return Node!(execute, ReturnValue:Movie Pipeline PIEExecutor Object Reference=@n15)
   Add(TargetMap:"Map of Strings to Float (single-precision)s"=@n2, Value:"Float (single-precision) (by ref)")
   Values!(execute, TargetMap:Map of Wildcards to Wildcards) -> (then, Values:Array of Wildcards)
   Add(execute=@n1.then, TargetMap=@n2, Key=r.Lumen.Reflections.Allow)
   GetCheckedState!(self:Check Box Object Reference=Get EnableTAA_Checkbox() -> (EnableTAA_Checkbox:Check Box Object Reference)) -> (ReturnValue:ECheckBoxState Enum=Unchecked)
   Get MRQ_Sequence()
   Get OutputDirectory()
  ]}
 ResetPreviewOverrides {nodes [
   ResetPreviewOverrides() {is=title K2Node_FunctionEntry}
   n1 = Get Seams()
   > Set labelColor_Default!(execute, labelColor_Default:Linear Color Structure=(R=1,G=1,B=1,A=1), self:TA Base Editor Button Widget Object Reference=@n1) -> (then, Output_Get:Linear Color Structure)
   n2 = Get Greybox()
   > Set labelColor_Default(labelColor_Default=(R=1,G=1,B=1,A=1), self=@n2)
   n3 = Get AoMap()
   > Set labelColor_Default(labelColor_Default=(R=1,G=1,B=1,A=1), self=@n3)
   n4 = Get Wireframe()
   > Set labelColor_Default(labelColor_Default=(R=1,G=1,B=1,A=1), self=@n4)
   n5 = Get VertexColor()
   > Set labelColor_Default(labelColor_Default=(R=1,G=1,B=1,A=1), self=@n5)
   n6 = Get Checker()
   > Set labelColor_Default(labelColor_Default=(R=1,G=1,B=1,A=1), self=@n6)
   > Set labelColor_Selected!(execute, labelColor_Selected:Linear Color Structure, self:TA Base Editor Button Widget Object Reference=@n1) -> (then, Output_Get:Linear Color Structure)
   > Set labelColor_Selected(self=@n2)
   > Set buttonColor_Default!(execute, buttonColor_Default:Linear Color Structure=(R=0.1219,G=0.968,B=1,A=1), self:TA Base Editor Button Widget Object Reference=@n1) -> (then, Output_Get:Linear Color Structure)
   > Set buttonColor_Default(buttonColor_Default=(R=0.0897,G=1,B=0.9384,A=1), self=@n2)
   > Set buttonColor_Default(buttonColor_Default=(R=0.0897,G=1,B=0.9384,A=1), self=@n3)
   > Set buttonColor_Default(buttonColor_Default=(R=0.0897,G=1,B=0.9384,A=1), self=@n4)
   > Set buttonColor_Default(buttonColor_Default=(R=0.0897,G=1,B=0.9384,A=1), self=@n5)
   > Set buttonColor_Default(buttonColor_Default=(R=0.0897,G=1,B=0.9384,A=1), self=@n6)
  ]}
 SetOverrideColor {nodes [
   n1 = SetOverrideColor!() -> (then, Target:TA Base Editor Button Widget Object Reference, labelColor_Default:Linear Color Structure) {is=title K2Node_FunctionEntry}
   > Set buttonColor_Default(buttonColor_Default=@n1.labelColor_Default, self=@n1)
   Set labelColor_Default(labelColor_Default=@n1.labelColor_Default, self=@n1)
  ]}
 SetPreviewOverride {nodes [SetPreviewOverride() {is=title K2Node_FunctionEntry}]}
 SetSuffixString {nodes [
   n1 = SetSuffixString!() -> (
    then
    Type:LPTT_ImportPart_PathGen Enum=NewEnumerator0
    ImportEntry:"LPTT Import Part Struct Structure (by ref)"
    UnknownName:Name
    Dilate:int
    Advanced Settings Pivot Style:LPTT_PivotLocation_Enum Enum=NewEnumerator0
   ) {is=title K2Node_FunctionEntry}
   > Set members in LPTT Import Part Struct!(execute, StructRef=@n1.ImportEntry, AssetType_7_#2=@n1) -> (then, StructOut)
   > n2 = Switch on LPTT_ImportPart_PathGen(Selection=@n1)
   Set members in LPTT Import Part Struct!(
    execute=@n2.NewEnumerator1
    StructRef=@n1.ImportEntry
    AdvancedSettings_30_#4_WireframeDilateOverride_5_#5:int=@n1.Dilate
    AdvancedSettings_30_#4_PivotStyle_8_#6:LPTT_PivotLocation_Enum Enum=@n1.Advanced Settings Pivot Style
    AdvancedSettings_30_#4_ForceCreateWireframe_18_#7:bool=False
    AdvancedSettings_30_#4_ForceCreateColorized_UVs_21_#8:bool=False
    AdvancedSettings_30_#4_ForceCreateSeams_22_#9:bool=False
   ) -> (then, StructOut)
   Set members in LPTT Import Part Struct(
    execute=@n2.NewEnumerator10
    StructRef=@n1.ImportEntry
    AdvancedSettings_30_#4_WireframeDilateOverride_5_#5=@n1.Dilate
    AdvancedSettings_30_#4_PivotStyle_8_#6=@n1.Advanced Settings Pivot Style
    AdvancedSettings_30_#4_ForceCreateWireframe_18_#7=False
    AdvancedSettings_30_#4_ForceCreateColorized_UVs_21_#8=False
    AdvancedSettings_30_#4_ForceCreateSeams_22_#9=False
   )
   Set members in LPTT Import Part Struct(
    execute=@n2.NewEnumerator2
    StructRef=@n1.ImportEntry
    AdvancedSettings_30_#4_WireframeDilateOverride_5_#5=@n1.Dilate
    AdvancedSettings_30_#4_PivotStyle_8_#6=@n1.Advanced Settings Pivot Style
    AdvancedSettings_30_#4_ForceCreateWireframe_18_#7=False
    AdvancedSettings_30_#4_ForceCreateColorized_UVs_21_#8=False
    AdvancedSettings_30_#4_ForceCreateSeams_22_#9=False
   )
   Set members in LPTT Import Part Struct(
    execute=@n2.NewEnumerator3
    StructRef=@n1.ImportEntry
    AdvancedSettings_30_#4_WireframeDilateOverride_5_#5=@n1.Dilate
    AdvancedSettings_30_#4_PivotStyle_8_#6=@n1.Advanced Settings Pivot Style
    AdvancedSettings_30_#4_ForceCreateWireframe_18_#7=False
    AdvancedSettings_30_#4_ForceCreateColorized_UVs_21_#8=False
    AdvancedSettings_30_#4_ForceCreateSeams_22_#9=False
   )
   Set members in LPTT Import Part Struct(
    execute=@n2.NewEnumerator4
    StructRef=@n1.ImportEntry
    AdvancedSettings_30_#4_WireframeDilateOverride_5_#5=@n1.Dilate
    AdvancedSettings_30_#4_PivotStyle_8_#6=@n1.Advanced Settings Pivot Style
    AdvancedSettings_30_#4_ForceCreateWireframe_18_#7=False
    AdvancedSettings_30_#4_ForceCreateColorized_UVs_21_#8=False
    AdvancedSettings_30_#4_ForceCreateSeams_22_#9=False
   )
   Set members in LPTT Import Part Struct(
    execute=@n2.NewEnumerator7
    StructRef=@n1.ImportEntry
    AdvancedSettings_30_#4_WireframeDilateOverride_5_#5=@n1.Dilate
    AdvancedSettings_30_#4_PivotStyle_8_#6=@n1.Advanced Settings Pivot Style
    AdvancedSettings_30_#4_ForceCreateWireframe_18_#7=False
    AdvancedSettings_30_#4_ForceCreateColorized_UVs_21_#8=False
    AdvancedSettings_30_#4_ForceCreateSeams_22_#9=False
   )
   Set members in LPTT Import Part Struct(
    execute=@n2.NewEnumerator8
    StructRef=@n1.ImportEntry
    AdvancedSettings_30_#4_WireframeDilateOverride_5_#5=@n1.Dilate
    AdvancedSettings_30_#4_PivotStyle_8_#6=@n1.Advanced Settings Pivot Style
    AdvancedSettings_30_#4_ForceCreateWireframe_18_#7=False
    AdvancedSettings_30_#4_ForceCreateColorized_UVs_21_#8=False
    AdvancedSettings_30_#4_ForceCreateSeams_22_#9=False
   )
  ]}
 ShowSpotterSpheres? {nodes [
   n1 = ShowSpotterSpheres?!() -> (then, Show?:bool=~) {is=title K2Node_FunctionEntry}
   > n2 = GetAllActorsWithTag!(execute, WorldContextObject:Object Reference, Tag:Name=Spotter) -> (then, OutActors:Array of Actor Object References)
   > n3 = For Each Loop(Array:Array of Actor Object References=@n2) -> (Array Element:Actor Object Reference)
   > n4 = Branch(Condition=@n1)
   > n5 = Branch(Condition=ActorHasTag!(self:Actor Object Reference=@n3, Tag:Name=Zoom) -> (ReturnValue:bool))
   n6 = NOT Boolean(A=Get RenderZoomed())
   > SetIsTemporarilyHiddenInEditor!(execute, self:Actor Object Reference=@n3, bIsHidden:bool=@n6) -> (then)
   > Set Actor Hidden In Game!(execute, self:Actor Object Reference=@n3, bNewHidden:bool=@n6) -> (then) {is=title K2Node_CallFunction}
   Branch(execute=@n5.else, Condition=ActorHasTag(self=@n3, Tag=Head))
   n7 = NOT Boolean(A=Get FocusHead())
   > SetIsTemporarilyHiddenInEditor(self=@n3, bIsHidden=@n7)
   > Set Actor Hidden In Game(self=@n3, bNewHidden=@n7)
   SetIsTemporarilyHiddenInEditor(execute=@n4.else, self=@n3, bIsHidden=true)
   > Set Actor Hidden In Game(self=@n3, bNewHidden=true)
  ]}
 Spawn ActorSet {nodes [
   n1 = Spawn ActorSet!() -> (then, SpawnedActorArrayRef:Array of LPTT Static Mesh Actor Object References, Spawn Transform Location:Vector=~, ShowControls?:bool=~) {is=title K2Node_FunctionEntry}
   > n2 = For Each Loop!(Exec, Array:Array of LPTT Part Data Struct Structures=Get CompiledPartData()) -> (
    LoopBody
    Array Element_FBXPath_29_#11:File Path Structure
    Array Element_Mesh_5_#12:Static Mesh Object Reference
    Array Element_Textures_30_#13:Array of Texture 2D Object References
    Array Element_MaterialSlotNames_24_#14:Array of LPTT Import Material Slot Struct Structures
    Array Element_PartType_25_#15:LPTT_ImportPart_PathGen Enum
    Array Element_WireframeDilateSize_12_#16:int
    Array Element_PivotStyle_15_#17:LPTT_PivotLocation_Enum Enum
    Array Element_ForceCreateWireframe_32_#18:bool=~
    Array Element_ForceCreateColorized_UVs_34_#19:bool=~
    Array Element_ForceCreateSeams_36_#20:bool=~
    Array Index
    Completed
   )
   Return Node!(
    execute=@n2.Completed
    ThisActorSet:Array of LPTT Static Mesh Actor Object References=Get LocalSpawnedActors() -> (LocalSpawnedActors:Array of LPTT Static Mesh Actor Object References)
   )
   > n3 = Set SpawnLocation(SpawnLocation:Vector=@n1.Spawn Transform Location) -> (Output_Get:Vector)
   n4 = SpawnActorFromClass!(
    execute=[@n10.else, @n3.then, @n11.then]
    ActorClass:Actor Class Reference=&LPTT_StaticMeshActor_C
    Location:Vector=Get SpawnLocation() -> (SpawnLocation:Vector)
    Rotation:Rotator="0,90,0"
    bTransient:bool
   ) -> (then, ReturnValue:LPTT Static Mesh Actor Object Reference)
   > Initialize!(
    execute
    self:LPTT Static Mesh Actor Object Reference=@n4
    Mesh:Static Mesh Object Reference=@n2.Array Element_Mesh_5_#12
    FBX_FilePath:File Path Structure=@n2
    Textures:Array of Texture 2D Object References=@n2.Array Element_Textures_30_#13
    MaterialSlots:Array of LPTT Import Material Slot Struct Structures=@n2.Array Element_MaterialSlotNames_24_#14
    NormalsImportSettings:Enum_LPTT_NormalsImportSettings Enum=Get Normals Import Setting() -> (Normals Import Setting=NewEnumerator0)
    DispName:String=GetDisplayName(Object=@n2.Array Element_Mesh_5_#12)
    AssetName:String=Get AssetName()
    DilateSize:int=@n2.Array Element_WireframeDilateSize_12_#16
    PartType:LPTT_ImportPart_PathGen Enum=@n2.Array Element_PartType_25_#15
    PivotStyle:LPTT_PivotLocation_Enum Enum=@n2.Array Element_PivotStyle_15_#17
    ForceWireframe:bool=@n2.Array Element_ForceCreateWireframe_32_#18
    ForceColorizedUVs:bool=@n2.Array Element_ForceCreateColorized_UVs_34_#19
    ForceSeams:bool=@n2.Array Element_ForceCreateSeams_36_#20
    QuickLoad:bool=Get Quickload()
   ) -> (then)
   > AddActorTag(Actor=@n4, Tag=String To Name(InString=Enum to String(Enumerator=@n2.Array Element_PartType_25_#15)))
   > Add!(
    execute
    TargetArray:Array of LPTT Static Mesh Actor Object References=Get LocalSpawnedActors()
    NewItem:"LPTT Static Mesh Actor Object Reference (by ref)"=@n4
   ) -> (then, ReturnValue) {is=title K2Node_CallArrayFunction}
   > n5 = Create Widget!(
    execute:Exec
    WorldContextObject:Object Reference
    Class:User Widget Class Reference=/Game/NeverCook/ArtTools/Tech_Art/LowPolyTurntable/Data/EUW_LPTT_PartData.EUW_LPTT_PartData_C
    OwningPlayer:Player Controller Object Reference
    ActorRefToWatch:Actor Object Reference=@n4
   ) -> (then:Exec, ReturnValue:EUW LPTT Part Data Object Reference)
   > n6 = Branch(Condition=@n1.ShowControls?)
   Add(
    execute=[@n9.then, @n6.then]
    TargetArray:Array of EUW LPTT Part Data Object References=Get PartDataSubWidgets()
    NewItem:"EUW LPTT Part Data Object Reference (by ref)"=@n5
   )
   > n7 = Branch(Condition=@n1.ShowControls?)
   > n8 = AddChild(self=Get WrapBox_0(), Content=@n5)
   Add(
    execute=[@n8.then, @n7.else]
    TargetArray:Array of LPTT Static Mesh Actor Object References=@n1
    NewItem:"LPTT Static Mesh Actor Object Reference (by ref)"=@n4
   )
   n9 = ToggleControlVisibility(execute=@n6.else, self:EUW LPTT Part Data Object Reference=@n5)
   n10 = Branch(Condition=OR Boolean(
     A="Equal (Enum)"(A:LPTT_ImportPart_PathGen Enum=@n2.Array Element_PartType_25_#15, B:LPTT_ImportPart_PathGen Enum=NewEnumerator0)
     B="Equal (Enum)"(A=@n2.Array Element_PartType_25_#15, B=NewEnumerator0)
    ))
   > n11 = Set SpawnLocation(SpawnLocation=vector + vector(A=Get SpawnLocation(), B="0,0,90"))
   Reroute Node(InputPin:LPTT_PivotLocation_Enum Enum=@n2.Array Element_PivotStyle_15_#17) -> (OutputPin:LPTT_PivotLocation_Enum Enum)
  ]}
 UpdateImportPropertyString {nodes [
   UpdateImportPropertyString() {is=title K2Node_FunctionEntry}
   > Clear(TargetArray:Array of Strings=Get ImportStructsStringCurrent())
   > n1 = For Each Loop!(Exec, Array:Array of LPTT Import Part Struct Structures=Get ImportPaths()) -> (LoopBody, Array Element:LPTT Import Part Struct Structure, Array Index, Completed)
   Set ImportStructsStringLast(execute=@n1.Completed, ImportStructsStringLast:Array of Strings=Get ImportStructsStringCurrent()) -> (Output_Get:Array of Strings)
   n2 = Convert Struct To Json String!(execute:Exec=@n1.LoopBody, Struct:LPTT Import Part Struct Structure=@n1) -> (then:Exec, OutJsonString:String, ReturnValue:bool) {is=title K2Node_StructToJsonString}
   > Add(TargetArray:Array of Strings=Get ImportStructsStringCurrent(), NewItem:"String (by ref)"=@n2)
  ]}
 UpdateMaterialSlots {nodes [
   n1 = UpdateMaterialSlots!() -> (then, ImportPartStruct:"LPTT Import Part Struct Structure (by ref)") {is=title K2Node_FunctionEntry}
   > n2 = GetFBX(File=Break LPTT Import Part Struct(LPTT_ImportPartStruct=@n1), PartType=IMPORTCHECK, NormalsImportSetting=NewEnumerator0) {is=title K2Node_CallFunction}
   > Set TEMPDEBUG(TEMPDEBUG=@n2)
   > n3 = Execute Python Script!(
    execute
    PythonScript="import unreal\r\n\r\nsm_component = unreal.StaticMeshComponent()\r\nsm_component.set_static_mesh(sm_asset)\r\nslot_names = sm_component.get_material_slot_names()"
    sm_asset:Static Mesh Object Reference=@n2
   ) -> (then, ReturnValue, slot_names:Array of Names) {inputs [sm_asset], outputs [slot_names]}
   > n4 = Set names(names=@n3.slot_names)
   > n5 = For Each Loop(Array:Array of Names=@n4) -> (Array Element:Name)
   Set members in LPTT Import Part Struct!(
    execute=@n5.Completed
    StructRef=Get ImportPartStruct() -> (ImportPartStruct:"LPTT Import Part Struct Structure (by ref)")
    MaterialSlots_40_#3:Array of LPTT Import Material Slot Struct Structures=Get MaterialSlots()
   ) -> (then, StructOut)
   n6 = "To String (Name)"(InName=@n5)
   n7 = DetectMaterialSlotType(execute=@n5.LoopBody, SlotName=@n6, PartType=Get DetectedType() -> (DetectedType=NewEnumerator0)) -> (Type=NewEnumerator0) {is=title K2Node_CallFunction}
   > Add(
    TargetArray:Array of LPTT Import Material Slot Struct Structures=Get MaterialSlots()
    NewItem:"LPTT Import Material Slot Struct Structure (by ref)"=Make LPTT Import Material Slot Struct(SlotName_23_#10=@n6, MaterialType_26_#6=@n7)
   )
  ]}
]
variables [
 ActiveTargetRotator {type=BP_LPTT_TargetPivotController_C, default=null}
 AssetName {type=string, default=""}
 AutoFindOthers? {type=boolean}
 BakeMapsPath {type="array<DirectoryPath>", default [{path=""}]}
 CompiledPartData {type="array<LPTT_PartDataStruct>", default []}
 CustomInterchange {type=SoftObjectPath, default=&/LPTT_InterchangePipeline}
 EnableTAA_Checkbox {type=CheckBox, default=null}
 Filepath {type=FilePath, default {filePath=""}}
 FocusHead {type=boolean}
 GlobalImportShift {type=Vector}
 ImportDirectory {type=DirectoryPath, default {path=""}}
 ImportedFBX {type="map<StaticMesh>", default {}}
 ImportPaths {type="array<LPTT_ImportPartStruct>", default []}
 ImportPropertyIgnores {type="array<string>", default [SlotName, MaterialType, CustomMaterialType, ForceCreateWireframe, ForceCreateColorized_UVs, ForceCreateSeams]}
 ImportSets {type="array<string>", default []}
 ImportStructsStringCurrent {type="array<string>", default []}
 ImportStructsStringLast {type="array<string>", default []}
 InputArray_FBX {type="array<string>", default []}
 InputArray_Texture {type="array<string>", default []}
 ModifiedImportParameterName {type=string, default=""}
 MRQ_Sequence {type=SoftObjectPath, default=/Game/NeverCook/ArtTools/Tech_Art/LowPolyTurntable/Data/LowPolyTT_Seq}
 MRQ_Sequence_HeadFocus {type=SoftObjectPath, default=/Game/NeverCook/ArtTools/Tech_Art/LowPolyTurntable/Data/LowPolyTT_HeadFocus}
 MRQ_Sequence_Zoomed {type=SoftObjectPath, default=/Game/NeverCook/ArtTools/Tech_Art/LowPolyTurntable/Data/LowPolyTT_Seq_Zoomed}
 MRQPresets {type=MoviePipelinePrimaryConfig, default=/Game/NeverCook/ArtTools/Tech_Art/LowPolyTurntable/Data/LPTT_MRQSettings}
 Normals Import Setting {type=Enum_LPTT_NormalsImportSettings, default=ComputeNormalsAndTangents}
 OutputDirectory {type=DirectoryPath, default {path=""}}
 PartDataSubWidgets {type="array<EUW_LPTT_PartData_C>", default []}
 PartPathGroups {type="array<LPTT_FilenameImportStruct>", default []}
 PartTypePathHints {
  type="map<LPTT_ImportPart_PathGen>"
  default {
   assets/Backpack=Backpack
   assets/Companion=Sidekick
   assets/Face Accessory=FaceAcc
   assets/Head=Head
   assets/Outfit=Body
   assets/Shoes=Shoe
   "assets\\Backpack"=Backpack
   "assets\\Companion"=Sidekick
   "assets\\Face Accessory"=FaceAcc
   "assets\\Head"=Head
   "assets\\Outfit"=Body
   "assets\\Shoes"=Shoe
  }
 }
 PartTypes {type="array<string>", default []}
 Quickload {type=boolean}
 RenderZoomed {type=boolean}
 SaveDataSlot {type=string, default=LPTTToolSaveData}
 SelfRef {type=Object, default=&self}
 ShouldLoadSavedSettings {type=boolean, default=true}
 ShouldUpdateImportProperties {type=boolean, default=true}
 SpawnedActors {type="array<LPTT_StaticMeshActor_C>", default []}
 TEMPDEBUG {type=StaticMesh, default=null}
 TextureExtensions {type="array<string>", default [_AO,_CRV,_N,_VC]}
 TextureExtensionsIgnore {type="array<string>", default [_AOC,_AO_Contact]}
 TransformOverrides {type=LPTT_TransformOverrideStruct, default {scale=1, location=<0 0 0>, rotation=<rot 0 0 0>}}
]
widgets {
 info {namedSlotCount=2, rootWidgetClass=&Overlay, widgetCount=31}
 widgets [
  Overlay_29 {widgetClassPath=&Overlay, widget=Overlay_29}
  EditorUtilityScrollBox_0 {
   widgetClassPath=&EditorUtilityScrollBox
   parent=Overlay_29
   widget=EditorUtilityScrollBox_0
   slot {class=OverlaySlot, properties {horizontalAlignment=HAlign_Fill}}
  }
  VerticalBox_0 {widgetClassPath=&VerticalBox, parent=EditorUtilityScrollBox_0, widget=VerticalBox_0, slot {class=ScrollBoxSlot}}
  HorizontalBox_0 {widgetClassPath=&HorizontalBox, parent=VerticalBox_0, widget=HorizontalBox_0, slot {class=VerticalBoxSlot}}
  VerticalBox_1 {
   widgetClassPath=&VerticalBox
   parent=HorizontalBox_0
   widget=VerticalBox_1
   slot {class=HorizontalBoxSlot, properties {size {sizeRule=Fill, value=1}}}
  }
  HorizontalBox_1 {widgetClassPath=&HorizontalBox, parent=VerticalBox_1, widget=HorizontalBox_1, slot {class=VerticalBoxSlot}}
  VerticalBox_2 {
   widgetClassPath=&VerticalBox
   parent=HorizontalBox_1
   widget=VerticalBox_2
   slot {class=HorizontalBoxSlot, properties {size {sizeRule=Fill, value=0.1}}}
  }
  Importtotextures {
   widgetClassPath=&TA_Generic01_Button
   parent=VerticalBox_2
   widget=Importtotextures
   properties {label=Import}
   slot {class=VerticalBoxSlot, properties {size {sizeRule=Automatic, value=2}}}
  }
  Render {
   widgetClassPath=&TA_Generic01_Button
   parent=VerticalBox_2
   widget=Render
   properties {label=Render}
   slot {class=VerticalBoxSlot}
  }
  ClearImportValues {
   widgetClassPath=&TA_Generic01_Button
   parent=VerticalBox_2
   widget=ClearImportValues
   properties {label=Clear Import Settings, buttonColor_Default=<rgba 1 0 0.1775 1>}
   slot {class=VerticalBoxSlot, properties {padding=<ltrb 0 30 0 0>}}
  }
  ClearFloor {
   widgetClassPath=&TA_Generic01_Button
   parent=VerticalBox_2
   widget=ClearFloor
   properties {label=Clear Turntable Floor, buttonColor_Default=<rgba 1 0 0.1775 1>}
   slot {class=VerticalBoxSlot, properties {padding=<ltrb 0 5 0 0>}}
  }
  VerticalBox_58 {
   widgetClassPath=&VerticalBox
   parent=HorizontalBox_1
   widget=VerticalBox_58
   slot {class=HorizontalBoxSlot, properties {size {sizeRule=Fill, value=1}}}
  }
  AssetNameField {
   widgetClassPath=&DetailsView
   parent=VerticalBox_58
   widget=AssetNameField
   properties {propertiesToShow [AssetName]}
   slot {class=VerticalBoxSlot}
  }
  RenderSettings {
   widgetClassPath=&DetailsView
   parent=VerticalBox_58
   widget=RenderSettings
   properties {categoriesToShow [RenderSettings]}
   slot {class=VerticalBoxSlot, properties {size {sizeRule=Fill, value=1}}}
  }
  Import_Paths {
   widgetClassPath=&DetailsView
   parent=VerticalBox_0
   widget=Import_Paths
   properties {propertiesToShow [ImportPaths,BakeMapsPath]}
   slot {class=VerticalBoxSlot}
  }
  Import_Settings {
   widgetClassPath=&DetailsView
   parent=VerticalBox_0
   widget=Import_Settings
   properties {categoriesToShow [Import Settings]}
   slot {class=VerticalBoxSlot}
  }
  DV_Textures {
   widgetClassPath=&DetailsView
   parent=VerticalBox_0
   widget=DV_Textures
   properties {visibility=Hidden, categoriesToShow [Textures]}
   slot {class=VerticalBoxSlot}
  }
  EditorUtilityExpandableArea_186 {
   widgetClassPath=/Script/Blutility.EditorUtilityExpandableArea
   bIsVariable=true
   parent=VerticalBox_0
   widget=EditorUtilityExpandableArea_186
   properties {bIsExpanded=true}
   slot {class=VerticalBoxSlot}
  }
  TextBlock_82 {
   widgetClassPath=&TextBlock
   namedSlotHost=EditorUtilityExpandableArea_186
   widget=TextBlock_82
   properties {
    text=Debug Section
    font {
     bForceMonospaced=false
     bMaterialIsStencil=false
     fontMaterial=null
     fontObject=&Roboto
     letterSpacing=0
     monospacedWidth=1
     size=15
     skewAmount=0
     typefaceFontName=Bold
     outlineSettings {
      bApplyOutlineToDropShadows=false
      bMiteredCorners=false
      bSeparateFillAlpha=false
      outlineMaterial=null
      outlineSize=0
      outlineColor=<rgba 0 0 0 1>
     }
    }
   }
  }
  VerticalBox_3 {widgetClassPath=&VerticalBox, namedSlotHost=EditorUtilityExpandableArea_186, widget=VerticalBox_3}
  HorizontalBox_2 {
   widgetClassPath=&HorizontalBox
   parent=VerticalBox_3
   widget=HorizontalBox_2
   slot {class=VerticalBoxSlot, properties {size {sizeRule=Automatic, value=2}}}
  }
  Greybox {
   widgetClassPath=&TA_Generic01_Button
   parent=HorizontalBox_2
   widget=Greybox
   properties {label=Grey}
   slot {class=HorizontalBoxSlot}
  }
  Wireframe {
   widgetClassPath=&TA_Generic01_Button
   parent=HorizontalBox_2
   widget=Wireframe
   properties {label=Wireframe}
   slot {class=HorizontalBoxSlot}
  }
  Seams {
   widgetClassPath=&TA_Generic01_Button
   parent=HorizontalBox_2
   widget=Seams
   properties {label=Seams}
   slot {class=HorizontalBoxSlot}
  }
  Checker {
   widgetClassPath=&TA_Generic01_Button
   parent=HorizontalBox_2
   widget=Checker
   properties {label=UV Checker}
   slot {class=HorizontalBoxSlot}
  }
  VertexColor {
   widgetClassPath=&TA_Generic01_Button
   parent=HorizontalBox_2
   widget=VertexColor
   properties {label=Vertex Color}
   slot {class=HorizontalBoxSlot}
  }
  AoMap {
   widgetClassPath=&TA_Generic01_Button
   parent=HorizontalBox_2
   widget=AoMap
   properties {label=Ao}
   slot {class=HorizontalBoxSlot}
  }
  Rotation_Slider {
   widgetClassPath=/Script/UMG.Slider
   bIsVariable=true
   parent=VerticalBox_3
   widget=Rotation_Slider
   properties {
    maxValue=180
    minValue=-180
    toolTipText=Rotates the Active Actor Rotator
    widgetStyle {
     barThickness=4
     disabledBarImage {
      bIsDynamicallyLoaded=false
      drawAs=RoundedBox
      imageType=NoImage
      mirroring=NoMirror
      resourceName=None
      resourceObject=null
      tiling=NoTile
      imageSize=<32 32>
      margin=<ltrb 0 0 0 0>
      outlineSettings {
       bUseBrushTransparency=false
       roundingType=FixedRadius
       width=1
       color {colorUseRule=UseColor_ColorTable, specifiedColor=<rgba 0 0 0 0>}
       cornerRadii=<2 2 2 2>
      }
      tintColor {colorUseRule=UseColor_Specified, specifiedColor=<rgba 1 1 1 1>}
      uVRegion {bIsValid=false, max=<0 0>, min=<0 0>}
     }
     disabledThumbImage {
      bIsDynamicallyLoaded=false
      drawAs=RoundedBox
      imageType=NoImage
      mirroring=NoMirror
      resourceName=None
      resourceObject=null
      tiling=NoTile
      imageSize=<32 32>
      margin=<ltrb 0 0 0 0>
      outlineSettings {
       bUseBrushTransparency=false
       roundingType=HalfHeightRadius
       width=0
       color {colorUseRule=UseColor_Specified, specifiedColor=<rgba 0 0 0 0>}
       cornerRadii=<0 0 0 0>
      }
      tintColor {colorUseRule=UseColor_Specified, specifiedColor=<rgba 0.5271 0.5271 0.5271 1>}
      uVRegion {bIsValid=false, max=<0 0>, min=<0 0>}
     }
     hoveredBarImage {
      bIsDynamicallyLoaded=false
      drawAs=RoundedBox
      imageType=NoImage
      mirroring=NoMirror
      resourceName=None
      resourceObject=null
      tiling=NoTile
      imageSize=<0 0>
      margin=<ltrb 0 0 0 0>
      outlineSettings {
       bUseBrushTransparency=false
       roundingType=FixedRadius
       width=1
       color {colorUseRule=UseColor_ColorTable, specifiedColor=<rgba 0 0 0 0>}
       cornerRadii=<2 2 2 2>
      }
      tintColor {colorUseRule=UseColor_Specified, specifiedColor=<rgba 0.0048 0.0048 0.0048 1>}
      uVRegion {bIsValid=false, max=<0 0>, min=<0 0>}
     }
     hoveredThumbImage {
      bIsDynamicallyLoaded=false
      drawAs=RoundedBox
      imageType=NoImage
      mirroring=NoMirror
      resourceName=None
      resourceObject=null
      tiling=NoTile
      imageSize=<8 8>
      margin=<ltrb 0 0 0 0>
      outlineSettings {
       bUseBrushTransparency=false
       roundingType=HalfHeightRadius
       width=0
       color {colorUseRule=UseColor_Specified, specifiedColor=<rgba 0 0 0 0>}
       cornerRadii=<0 0 0 0>
      }
      tintColor {colorUseRule=UseColor_Specified, specifiedColor=<rgba 1 1 1 1>}
      uVRegion {bIsValid=false, max=<0 0>, min=<0 0>}
     }
     normalBarImage {
      bIsDynamicallyLoaded=false
      drawAs=RoundedBox
      imageType=NoImage
      mirroring=NoMirror
      resourceName=None
      resourceObject=null
      tiling=NoTile
      imageSize=<0 0>
      margin=<ltrb 0 0 0 0>
      outlineSettings {
       bUseBrushTransparency=false
       roundingType=FixedRadius
       width=1
       color {colorUseRule=UseColor_ColorTable, specifiedColor=<rgba 0 0 0 0>}
       cornerRadii=<2 2 2 2>
      }
      tintColor {colorUseRule=UseColor_Specified, specifiedColor=<rgba 0.0048 0.0048 0.0048 1>}
      uVRegion {bIsValid=false, max=<0 0>, min=<0 0>}
     }
     normalThumbImage {
      bIsDynamicallyLoaded=false
      drawAs=RoundedBox
      imageType=NoImage
      mirroring=NoMirror
      resourceName=None
      resourceObject=null
      tiling=NoTile
      imageSize=<8 8>
      margin=<ltrb 0 0 0 0>
      outlineSettings {
       bUseBrushTransparency=false
       roundingType=HalfHeightRadius
       width=0
       color {colorUseRule=UseColor_Specified, specifiedColor=<rgba 0 0 0 0>}
       cornerRadii=<0 0 0 0>
      }
      tintColor {colorUseRule=UseColor_Specified, specifiedColor=<rgba 0.5271 0.5271 0.5271 1>}
      uVRegion {bIsValid=false, max=<0 0>, min=<0 0>}
     }
    }
   }
   slot {class=VerticalBoxSlot, properties {padding=<ltrb 0 10 0 20>}}
  }
  DebugDetailView {
   widgetClassPath=&DetailsView
   parent=VerticalBox_3
   widget=DebugDetailView
   properties {categoriesToShow [ImportDebug]}
   slot {class=VerticalBoxSlot}
  }
  TextBlock_265 {
   widgetClassPath=&TextBlock
   parent=VerticalBox_3
   widget=TextBlock_265
   properties {
    text=Loaded parts:
    font {
     bForceMonospaced=false
     bMaterialIsStencil=false
     fontMaterial=null
     fontObject=&Roboto
     letterSpacing=0
     monospacedWidth=1
     size=11.25
     skewAmount=0
     typefaceFontName=Bold
     outlineSettings {
      bApplyOutlineToDropShadows=false
      bMiteredCorners=false
      bSeparateFillAlpha=false
      outlineMaterial=null
      outlineSize=0
      outlineColor=<rgba 0 0 0 1>
     }
    }
   }
   slot {class=VerticalBoxSlot}
  }
  WrapBox_0 {
   widgetClassPath=&WrapBox
   bIsVariable=true
   parent=VerticalBox_3
   widget=WrapBox_0
   properties {horizontalAlignment=HAlign_Fill}
   slot {class=VerticalBoxSlot}
  }
 ]
}
